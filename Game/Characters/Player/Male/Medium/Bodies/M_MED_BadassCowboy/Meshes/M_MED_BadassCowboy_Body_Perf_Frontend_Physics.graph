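# Physics asset for the medium male BadassCowboy body's frontend variant, covering a dynamic long jacket, boot spurs and lasso root.
# Pelvis, leg and foot bodies anchor limited-angle front, side and rear jacket chains via thigh-twist and pelvis-linked master bodies.
# Extra free-rotation constraints cross-link some neighboring panels, tying separate chains together.
# Spurs lock swing and spin freely on twist; the lasso has only a tiny pelvis-linked master body.
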PhysicsAsset /Game/Characters/Player/Male/Medium/Bodies/M_MED_BadassCowboy/Meshes/M_MED_BadassCowboy_Body_Perf_Frontend_Physics
bodies {
 calf_l [{shapeType=Capsule, length=34.75, radius=8.75, center=<-19.3279 -0.0009 0>, rotation=<rot 90 0 180>}]
 calf_r [{shapeType=Capsule, length=35.75, radius=7.75, center=<19.3279 0.0009 0>, rotation=<rot 90 0 0>}]
 dyn_jacket_front_01_l [{shapeType=Box, extentX=15, extentY=3, extentZ=14.5, center=<5 -0.0002 0.1112>}]
 dyn_jacket_front_01_r [{shapeType=Box, extentX=15, extentY=3, extentZ=15.5, center=<4.9628 -0.0402 -0.4298>, rotation=<rot 0 0 5.0001>}]
 dyn_jacket_front_02_l [{shapeType=Box, extentX=15, extentY=5.5, extentZ=16, center=<5 0 0>}]
 dyn_jacket_front_02_r [{shapeType=Box, extentX=15, extentY=3, extentZ=16.5, center=<4.9597 0 0>}]
 dyn_jacket_front_03_l [{shapeType=Capsule, length=10.25, radius=7.75, shapeName=dyn_jacket_front_03_l_capsule, center=<5 -5 0>, rotation=<rot -30 -0.0001 0>}]
 dyn_jacket_front_03_r [{shapeType=Capsule, length=10, radius=7.75, shapeName=dyn_jacket_front_03_r_capsule, center=<5 -5 0>, rotation=<rot 30 -0.0001 0>}]
 dyn_jacket_front_04_l [{shapeType=Capsule, length=10.25, radius=7.75, shapeName=dyn_jacket_front_04_l_capsule, center=<5 -5 0>, rotation=<rot -30 -0.0001 0>}]
 dyn_jacket_front_04_r [{shapeType=Capsule, length=10, radius=7.75, shapeName=dyn_jacket_front_04_r_capsule, center=<5 -5 0>, rotation=<rot 30 -0.0001 0>}]
 dyn_jacket_front_05_l [{shapeType=Capsule, length=12.86, radius=6.3634, shapeName=dyn_jacket_front_05_l_capsule, center=<5 -4 0>, rotation=<rot -40 0 0>}]
 dyn_jacket_front_05_r [{shapeType=Capsule, length=10, radius=7.75, shapeName=dyn_jacket_front_05_r_capsule, center=<5 -5 0>, rotation=<rot 30 -0.0001 0>}]
 dyn_jacket_rear_01_l [{shapeType=Box, extentX=12, extentY=5, extentZ=15, center=<5 0 0>, rotation=<rot 0 0 15.0009>}]
 dyn_jacket_rear_01_r [{shapeType=Box, extentX=12, extentY=6.5, extentZ=18, center=<6 -2 0>, rotation=<rot 0 0 -15>}]
 dyn_jacket_rear_02_l [{shapeType=Box, extentX=15, extentY=5, extentZ=16, center=<5 0 0>, rotation=<rot 0 0 15.0009>}]
 dyn_jacket_rear_02_r [{shapeType=Box, extentX=15, extentY=6.5, extentZ=20, center=<8 -2 0>, rotation=<rot 0 0 -15>}]
 dyn_jacket_rear_03_l [
  {shapeType=Box, extentX=15, extentY=6, extentZ=20, center=<7 -2 -10>, rotation=<rot 0 0 10>}
  {shapeType=Box, extentX=15, extentY=5, extentZ=15.5, center=<7 2 5>, rotation=<rot 0 0 35.0004>}
 ]
 dyn_jacket_rear_03_r [
  {shapeType=Box, extentX=15, extentY=6, extentZ=20, center=<7 -2 10>, rotation=<rot 0 0 -10>}
  {shapeType=Box, extentX=15, extentY=5, extentZ=18, center=<7 6 -7>, rotation=<rot 0 0 -45>}
 ]
 dyn_jacket_rear_04_l [
  {shapeType=Box, extentX=15, extentY=6, extentZ=20, center=<7 -2 -10>, rotation=<rot 0 0 10>}
  {shapeType=Box, extentX=15, extentY=5, extentZ=15.5, center=<7 2 5>, rotation=<rot 0 0 35.0004>}
 ]
 dyn_jacket_rear_04_r [
  {shapeType=Box, extentX=15, extentY=6, extentZ=20, center=<7 -2 10>, rotation=<rot 0 0 -10>}
  {shapeType=Box, extentX=15, extentY=5, extentZ=18, center=<7 6 -7>, rotation=<rot 0 0 -45>}
 ]
 dyn_jacket_rear_l_MSTR [{shapeType=Capsule, length=0.505, radius=0.505, rotation=<rot 0 -90 0>}]
 dyn_jacket_rear_r_MSTR [{shapeType=Capsule, length=0.505, radius=0.505, rotation=<rot 0 -90 0>}]
 dyn_jacket_side_01_l [{shapeType=Box, extentX=15, extentY=3, extentZ=17.5, center=<5 0 0>, rotation=<rot 0 0 -15.0005>}]
 dyn_jacket_side_01_r [{shapeType=Box, extentX=15, extentY=4, extentZ=15.5, center=<5 -0.2345 0.0853>, rotation=<rot 0 0 10.0009>}]
 dyn_jacket_side_02_l [{shapeType=Box, extentX=15, extentY=3.5, extentZ=17.5, center=<5 0 0>, rotation=<rot 0 0 -15.0005>}]
 dyn_jacket_side_02_r [{shapeType=Box, extentX=15, extentY=4, extentZ=16.5, center=<5 0 0>, rotation=<rot 0 0 15.0009>}]
 dyn_jacket_side_03_l [
  {shapeType=Box, extentX=15, extentY=5, extentZ=14.5, center=<7 3 -6>, rotation=<rot 0 0 -30>}
  {shapeType=Box, extentX=15, extentY=5.5, extentZ=15, center=<7 0 6>}
 ]
 dyn_jacket_side_03_r [
  {shapeType=Box, extentX=15, extentY=5, extentZ=17.5, center=<7 6 6>, rotation=<rot 0 0 40>}
  {shapeType=Box, extentX=15, extentY=5, extentZ=15, center=<7 0 -6>}
 ]
 dyn_jacket_side_04_l [
  {shapeType=Box, extentX=15, extentY=5, extentZ=14.5, center=<7 3 -6>, rotation=<rot 0 0 -30>}
  {shapeType=Box, extentX=15, extentY=5.5, extentZ=15, center=<7 0 6>}
 ]
 dyn_jacket_side_04_r [
  {shapeType=Box, extentX=15, extentY=5, extentZ=17.5, center=<7 6 6>, rotation=<rot 0 0 40>}
  {shapeType=Box, extentX=15, extentY=5, extentZ=15, center=<7 0 -6>}
 ]
 dyn_lasso_MSTR [{shapeType=Capsule, length=0.505, radius=0.505, rotation=<rot 0 -90 0>}]
 dyn_spur_l [{shapeType=Capsule, length=0.0001, radius=2, center=<0 0.03 0.03>}]
 dyn_spur_r [{shapeType=Capsule, length=0.0001, radius=2, center=<0 0.03 0.03>}]
 foot_l [
  {shapeType=Capsule, length=14.75, radius=9.5, center=<-3.926 5.7336 0>, rotation=<rot 90 0 90>}
  {shapeType=Capsule, length=7.5, radius=5.5, center=<3.2048 -6.4887 -0.9887>, rotation=<rot 90 0 90>}
 ]
 foot_r [
  {shapeType=Capsule, length=7, radius=6, center=<-3.2048 6.4414 -0.9886>, rotation=<rot 90 0 90>}
  {shapeType=Capsule, length=15, radius=9.25, center=<3.9317 -5.6438 -0.2456>, rotation=<rot 90 0 90>}
 ]
 pelvis [{shapeType=Capsule, length=8.3835, radius=13.6435, center=<-12.7015 1.8555 0.6992>, rotation=<rot 0 0 -5.0005>}]
 thigh_l [{shapeType=Capsule, length=25.75, radius=9.5, center=<-25.0001 0.4707 1.0913>, rotation=<rot 90 0 -179.9999>}]
 thigh_r [{shapeType=Capsule, length=25.75, radius=9.5, center=<25 -1.4625 -0.6208>, rotation=<rot 90 0 0>}]
 thigh_twist_01_l [{shapeType=Box, extentX=1, extentY=1, extentZ=1}]
 thigh_twist_01_r [{shapeType=Box, extentX=1, extentY=1, extentZ=1}]
}
constraints [
 {
  bone1Name=calf_r
  bone2Name=thigh_r
  swing1LimitDegrees=45
  swing1Motion=Free
  swing2LimitDegrees=45
  swing2Motion=Free
  twistLimitDegrees=45
  twistMotion=Free
 }
 {
  bone1Name=calf_l
  bone2Name=thigh_l
  swing1LimitDegrees=45
  swing1Motion=Free
  swing2LimitDegrees=45
  swing2Motion=Free
  twistLimitDegrees=45
  twistMotion=Free
 }
 {bone1Name=dyn_jacket_side_01_l, bone2Name=thigh_twist_01_l, swing1LimitDegrees=35, swing2LimitDegrees=45, twistLimitDegrees=5}
 {bone1Name=dyn_jacket_side_02_l, bone2Name=dyn_jacket_side_01_l, swing1LimitDegrees=35, swing2LimitDegrees=45, twistLimitDegrees=5}
 {bone1Name=dyn_jacket_side_03_l, bone2Name=dyn_jacket_side_02_l, swing1LimitDegrees=35, swing2LimitDegrees=15, twistLimitDegrees=5}
 {bone1Name=dyn_jacket_side_04_l, bone2Name=dyn_jacket_side_03_l, swing1LimitDegrees=35, swing2LimitDegrees=15, twistLimitDegrees=5}
 {bone1Name=dyn_jacket_front_01_l, bone2Name=thigh_twist_01_l, swing1LimitDegrees=45, swing2LimitDegrees=45, twistLimitDegrees=15}
 {bone1Name=dyn_jacket_front_02_l, bone2Name=dyn_jacket_front_01_l, swing1LimitDegrees=35, swing2LimitDegrees=10, twistLimitDegrees=5}
 {bone1Name=dyn_jacket_front_03_l, bone2Name=dyn_jacket_front_02_l, swing1LimitDegrees=35, swing2LimitDegrees=10, twistLimitDegrees=5}
 {bone1Name=dyn_jacket_front_04_l, bone2Name=dyn_jacket_front_03_l, swing1LimitDegrees=35, swing2LimitDegrees=35, twistLimitDegrees=5}
 {bone1Name=dyn_jacket_front_05_l, bone2Name=dyn_jacket_front_04_l, swing1LimitDegrees=35, swing2LimitDegrees=35, twistLimitDegrees=5}
 {bone1Name=dyn_jacket_rear_01_l, bone2Name=dyn_jacket_rear_l_MSTR, swing1LimitDegrees=35, swing2LimitDegrees=30, twistLimitDegrees=10}
 {bone1Name=dyn_jacket_rear_02_l, bone2Name=dyn_jacket_rear_01_l, swing1LimitDegrees=35, swing2LimitDegrees=30, twistLimitDegrees=10}
 {bone1Name=dyn_jacket_rear_03_l, bone2Name=dyn_jacket_rear_02_l, swing1LimitDegrees=35, swing2LimitDegrees=10, twistLimitDegrees=10}
 {bone1Name=dyn_jacket_rear_04_l, bone2Name=dyn_jacket_rear_03_l, swing1LimitDegrees=35, swing2LimitDegrees=10, twistLimitDegrees=10}
 {bone1Name=dyn_jacket_side_01_r, bone2Name=thigh_twist_01_r, swing1LimitDegrees=35, swing2LimitDegrees=5, twistLimitDegrees=10}
 {bone1Name=dyn_jacket_side_02_r, bone2Name=dyn_jacket_side_01_r, swing1LimitDegrees=35, swing2LimitDegrees=5, twistLimitDegrees=10}
 {bone1Name=dyn_jacket_side_03_r, bone2Name=dyn_jacket_side_02_r, swing1LimitDegrees=35, swing2LimitDegrees=5, twistLimitDegrees=10}
 {bone1Name=dyn_jacket_side_04_r, bone2Name=dyn_jacket_side_03_r, swing1LimitDegrees=35, swing2LimitDegrees=5, twistLimitDegrees=10}
 {bone1Name=dyn_jacket_front_01_r, bone2Name=thigh_twist_01_r, swing1LimitDegrees=45, swing2LimitDegrees=45, twistLimitDegrees=15}
 {bone1Name=dyn_jacket_front_02_r, bone2Name=dyn_jacket_front_01_r, swing1LimitDegrees=35, swing2LimitDegrees=10, twistLimitDegrees=15}
 {bone1Name=dyn_jacket_front_03_r, bone2Name=dyn_jacket_front_02_r, swing1LimitDegrees=35, swing2LimitDegrees=10, twistLimitDegrees=15}
 {bone1Name=dyn_jacket_front_04_r, bone2Name=dyn_jacket_front_03_r, swing1LimitDegrees=35, swing2LimitDegrees=10, twistLimitDegrees=15}
 {bone1Name=dyn_jacket_front_05_r, bone2Name=dyn_jacket_front_04_r, swing1LimitDegrees=35, swing2LimitDegrees=10, twistLimitDegrees=15}
 {bone1Name=dyn_jacket_rear_01_r, bone2Name=dyn_jacket_rear_r_MSTR, swing1LimitDegrees=35, swing2LimitDegrees=30, twistLimitDegrees=10}
 {bone1Name=dyn_jacket_rear_02_r, bone2Name=dyn_jacket_rear_01_r, swing1LimitDegrees=35, swing2LimitDegrees=30, twistLimitDegrees=10}
 {bone1Name=dyn_jacket_rear_03_r, bone2Name=dyn_jacket_rear_02_r, swing1LimitDegrees=35, swing2LimitDegrees=10, twistLimitDegrees=10}
 {bone1Name=dyn_jacket_rear_04_r, bone2Name=dyn_jacket_rear_03_r, swing1LimitDegrees=35, swing2LimitDegrees=10, twistLimitDegrees=10}
 {
  bone1Name=thigh_twist_01_r
  bone2Name=thigh_r
  swing1LimitDegrees=45
  swing1Motion=Free
  swing2LimitDegrees=45
  swing2Motion=Free
  twistLimitDegrees=45
  twistMotion=Free
 }
 {
  bone1Name=thigh_twist_01_l
  bone2Name=thigh_l
  swing1LimitDegrees=45
  swing1Motion=Free
  swing2LimitDegrees=45
  swing2Motion=Free
  twistLimitDegrees=45
  twistMotion=Free
 }
 {
  bone1Name=dyn_spur_l
  bone2Name=foot_l
  swing1LimitDegrees=45
  swing1Motion=Locked
  swing2LimitDegrees=45
  swing2Motion=Locked
  twistLimitDegrees=45
  twistMotion=Free
 }
 {
  bone1Name=dyn_spur_r
  bone2Name=foot_r
  swing1LimitDegrees=45
  swing1Motion=Locked
  swing2LimitDegrees=45
  swing2Motion=Locked
  twistLimitDegrees=45
  twistMotion=Free
 }
 {
  bone1Name=foot_r
  bone2Name=calf_r
  swing1LimitDegrees=45
  swing1Motion=Free
  swing2LimitDegrees=45
  swing2Motion=Free
  twistLimitDegrees=45
  twistMotion=Free
 }
 {
  bone1Name=foot_l
  bone2Name=calf_l
  swing1LimitDegrees=45
  swing1Motion=Free
  swing2LimitDegrees=45
  swing2Motion=Free
  twistLimitDegrees=45
  twistMotion=Free
 }
 {
  bone1Name=dyn_jacket_rear_02_r
  bone2Name=dyn_jacket_rear_02_l
  swing1LimitDegrees=45
  swing1Motion=Free
  swing2LimitDegrees=45
  swing2Motion=Free
  twistLimitDegrees=45
  twistMotion=Free
 }
 {
  bone1Name=thigh_l
  bone2Name=pelvis
  swing1LimitDegrees=45
  swing1Motion=Free
  swing2LimitDegrees=45
  swing2Motion=Free
  twistLimitDegrees=45
  twistMotion=Free
 }
 {
  bone1Name=thigh_r
  bone2Name=pelvis
  swing1LimitDegrees=45
  swing1Motion=Free
  swing2LimitDegrees=45
  swing2Motion=Free
  twistLimitDegrees=45
  twistMotion=Free
 }
 {
  bone1Name=dyn_jacket_rear_r_MSTR
  bone2Name=pelvis
  swing1LimitDegrees=45
  swing1Motion=Free
  swing2LimitDegrees=45
  swing2Motion=Free
  twistLimitDegrees=45
  twistMotion=Free
 }
 {
  bone1Name=dyn_jacket_rear_l_MSTR
  bone2Name=pelvis
  swing1LimitDegrees=45
  swing1Motion=Free
  swing2LimitDegrees=45
  swing2Motion=Free
  twistLimitDegrees=45
  twistMotion=Free
 }
 {
  bone1Name=dyn_jacket_front_03_r
  bone2Name=dyn_jacket_side_03_r
  swing1LimitDegrees=45
  swing1Motion=Free
  swing2LimitDegrees=45
  swing2Motion=Free
  twistLimitDegrees=45
  twistMotion=Free
 }
 {
  bone1Name=dyn_jacket_side_04_r
  bone2Name=dyn_jacket_rear_04_r
  swing1LimitDegrees=45
  swing1Motion=Free
  swing2LimitDegrees=5
  swing2Motion=Free
  twistLimitDegrees=45
  twistMotion=Free
 }
 {
  bone1Name=dyn_lasso_MSTR
  bone2Name=pelvis
  swing1LimitDegrees=45
  swing1Motion=Free
  swing2LimitDegrees=45
  swing2Motion=Free
  twistLimitDegrees=45
  twistMotion=Free
 }
 {
  bone1Name=dyn_jacket_side_04_l
  bone2Name=dyn_jacket_rear_04_l
  swing1LimitDegrees=45
  swing1Motion=Free
  swing2LimitDegrees=5
  swing2Motion=Free
  twistLimitDegrees=45
  twistMotion=Free
 }
 {
  bone1Name=dyn_jacket_front_04_l
  bone2Name=dyn_jacket_side_04_l
  swing1LimitDegrees=45
  swing1Motion=Free
  swing2LimitDegrees=45
  swing2Motion=Free
  twistLimitDegrees=45
  twistMotion=Free
 }
]
