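# Physics asset for the lasso on the medium male BadassCowboy player character, with simple bodies for the lasso, rope and nearby right-side body and jacket.
# The lasso chain hangs from a master body on the pelvis, and a rope chain branches from its first segment, both twist-locked below the master.
# Right-side jacket chains, thigh and tiny foot boxes attach through angularly free constraints, apparently serving as collision shapes.
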PhysicsAsset /Game/Characters/Player/Male/Medium/Bodies/M_MED_BadassCowboy/Meshes/M_MED_BadassCowboy_Lasso_Physics
bodies {
 dyn_jacket_front_01_r [{shapeType=Capsule, length=15, radius=10, shapeName=dyn_jacket_front_01_r_capsule, center=<5 9.2605 -1.2975>, rotation=<rot 90 0 0>}]
 dyn_jacket_front_02_r [{shapeType=Capsule, length=15, radius=10, shapeName=dyn_jacket_front_02_r_capsule, center=<5 9.2605 0>, rotation=<rot 90 0 0>}]
 dyn_jacket_front_03_r [{shapeType=Capsule, length=15, radius=10, shapeName=dyn_jacket_front_03_r_capsule, center=<5 9.2605 0>, rotation=<rot 90 0 0>}]
 dyn_jacket_rear_01_r [{shapeType=Capsule, length=11.5, radius=12.5, shapeName=dyn_jacket_rear_01_r_capsule, center=<5 10.9763 5.2656>, rotation=<rot 90 0 0>}]
 dyn_jacket_rear_02_r [{shapeType=Capsule, length=11.25, radius=13.75, shapeName=dyn_jacket_rear_02_r_capsule, center=<5 12.0218 4.9547>, rotation=<rot 90 0 0>}]
 dyn_jacket_rear_03_r [{shapeType=Capsule, length=11.25, radius=13.75, shapeName=dyn_jacket_rear_03_r_capsule, center=<5 12.2201 3.7164>, rotation=<rot 90 0 0>}]
 dyn_jacket_side_01_r [{shapeType=Capsule, length=15, radius=10, shapeName=dyn_jacket_side_01_r_capsule, center=<5 10.7786 0.532>, rotation=<rot 90 0 0>}]
 dyn_jacket_side_02_r [{shapeType=Capsule, length=15, radius=10, shapeName=dyn_jacket_side_02_r_capsule, center=<5 12.8354 0.7464>, rotation=<rot 90 0 0>}]
 dyn_jacket_side_03_r [{shapeType=Capsule, length=15, radius=10, shapeName=dyn_jacket_side_03_r_capsule, center=<5 12.0228 0.7464>, rotation=<rot 90 0 0>}]
 dyn_lasso_01 [{shapeType=Capsule, length=8, radius=2, center=<5 0 0>, rotation=<rot 90 0 0>}]
 dyn_lasso_02 [{shapeType=Capsule, length=8, radius=2, center=<5 0 0>, rotation=<rot 90 0 0>}]
 dyn_lasso_03 [{shapeType=Capsule, length=8, radius=2, center=<5 0 0>, rotation=<rot 90 0 0>}]
 dyn_lasso_MSTR [{shapeType=Capsule, length=2, radius=2}]
 dyn_rope_01 [{shapeType=Capsule, length=2.75, radius=2, center=<3 0 0>, rotation=<rot 90 0 0>}]
 dyn_rope_01_B [{shapeType=Capsule, length=2.75, radius=2, center=<4.2748 0 0>, rotation=<rot 90 0 0>}]
 dyn_rope_02 [{shapeType=Capsule, length=2.75, radius=2, center=<3.538 0 0>, rotation=<rot 90 165.7376 175.7377>}]
 dyn_rope_03 [{shapeType=Capsule, length=6.5, radius=2, center=<5 0 0>, rotation=<rot 90 0 0>}]
 foot_l [{shapeType=Box, extentX=1, extentY=1, extentZ=1}]
 foot_r [{shapeType=Box, extentX=1, extentY=1, extentZ=1}]
 pelvis [{shapeType=Capsule, length=9.75, radius=16.75, center=<-10.4994 0 0>, rotation=<rot -85 179.9999 180>}]
 thigh_r [{shapeType=Capsule, length=33.5, radius=11.25, center=<24.8179 0 -0.9843>, rotation=<rot 85 179.9999 179.9999>}]
}
constraints [
 {bone1Name=dyn_lasso_02, bone2Name=dyn_lasso_01, swing1LimitDegrees=10, swing2LimitDegrees=20, twistLimitDegrees=45, twistMotion=Locked}
 {bone1Name=dyn_lasso_03, bone2Name=dyn_lasso_02, swing1LimitDegrees=10, swing2LimitDegrees=20, twistLimitDegrees=45, twistMotion=Locked}
 {
  bone1Name=thigh_r
  bone2Name=pelvis
  swing1LimitDegrees=45
  swing1Motion=Free
  swing2LimitDegrees=45
  swing2Motion=Free
  twistLimitDegrees=45
  twistMotion=Free
 }
 {bone1Name=dyn_rope_01, bone2Name=dyn_lasso_01, swing1LimitDegrees=45, swing2LimitDegrees=45, twistLimitDegrees=45, twistMotion=Locked}
 {bone1Name=dyn_rope_01_B, bone2Name=dyn_rope_01, swing1LimitDegrees=45, swing2LimitDegrees=45, twistLimitDegrees=45, twistMotion=Locked}
 {bone1Name=dyn_rope_02, bone2Name=dyn_rope_01_B, swing1LimitDegrees=45, swing2LimitDegrees=45, twistLimitDegrees=45, twistMotion=Locked}
 {bone1Name=dyn_rope_03, bone2Name=dyn_rope_02, swing1LimitDegrees=45, swing2LimitDegrees=45, twistLimitDegrees=45, twistMotion=Locked}
 {bone1Name=dyn_lasso_01, bone2Name=dyn_lasso_MSTR, swing1LimitDegrees=2, swing2LimitDegrees=60, twistLimitDegrees=45, twistMotion=Locked}
 {
  bone1Name=foot_r
  bone2Name=thigh_r
  swing1LimitDegrees=45
  swing1Motion=Free
  swing2LimitDegrees=45
  swing2Motion=Free
  twistLimitDegrees=45
  twistMotion=Free
 }
 {
  bone1Name=foot_l
  bone2Name=pelvis
  swing1LimitDegrees=45
  swing1Motion=Free
  swing2LimitDegrees=45
  swing2Motion=Free
  twistLimitDegrees=45
  twistMotion=Free
 }
 {
  bone1Name=dyn_lasso_MSTR
  bone2Name=pelvis
  swing1LimitDegrees=45
  swing1Motion=Free
  swing2LimitDegrees=45
  swing2Motion=Free
  twistLimitDegrees=45
  twistMotion=Free
 }
 {
  bone1Name=dyn_jacket_side_02_r
  bone2Name=dyn_jacket_side_01_r
  swing1LimitDegrees=45
  swing1Motion=Free
  swing2LimitDegrees=45
  swing2Motion=Free
  twistLimitDegrees=45
  twistMotion=Free
 }
 {
  bone1Name=dyn_jacket_side_03_r
  bone2Name=dyn_jacket_side_02_r
  swing1LimitDegrees=45
  swing1Motion=Free
  swing2LimitDegrees=45
  swing2Motion=Free
  twistLimitDegrees=45
  twistMotion=Free
 }
 {
  bone1Name=dyn_jacket_rear_02_r
  bone2Name=dyn_jacket_rear_01_r
  swing1LimitDegrees=45
  swing1Motion=Free
  swing2LimitDegrees=45
  swing2Motion=Free
  twistLimitDegrees=45
  twistMotion=Free
 }
 {
  bone1Name=dyn_jacket_rear_03_r
  bone2Name=dyn_jacket_rear_02_r
  swing1LimitDegrees=45
  swing1Motion=Free
  swing2LimitDegrees=45
  swing2Motion=Free
  twistLimitDegrees=45
  twistMotion=Free
 }
 {
  bone1Name=dyn_jacket_rear_01_r
  bone2Name=pelvis
  swing1LimitDegrees=45
  swing1Motion=Free
  swing2LimitDegrees=45
  swing2Motion=Free
  twistLimitDegrees=45
  twistMotion=Free
 }
 {
  bone1Name=dyn_jacket_side_01_r
  bone2Name=thigh_r
  swing1LimitDegrees=45
  swing1Motion=Free
  swing2LimitDegrees=45
  swing2Motion=Free
  twistLimitDegrees=45
  twistMotion=Free
 }
 {
  bone1Name=dyn_jacket_front_01_r
  bone2Name=thigh_r
  swing1LimitDegrees=45
  swing1Motion=Free
  swing2LimitDegrees=45
  swing2Motion=Free
  twistLimitDegrees=45
  twistMotion=Free
 }
 {
  bone1Name=dyn_jacket_front_02_r
  bone2Name=dyn_jacket_front_01_r
  swing1LimitDegrees=45
  swing1Motion=Free
  swing2LimitDegrees=45
  swing2Motion=Free
  twistLimitDegrees=45
  twistMotion=Free
 }
 {
  bone1Name=dyn_jacket_front_03_r
  bone2Name=dyn_jacket_front_02_r
  swing1LimitDegrees=45
  swing1Motion=Free
  swing2LimitDegrees=45
  swing2Motion=Free
  twistLimitDegrees=45
  twistMotion=Free
 }
]
