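# Post-animation AnimBP for the male BadassCowboy player body, adding secondary motion to costume bones on a copied pose.
# Chained Anim Dynamics nodes spring-simulate jacket bones (side, front, rear) and partially weighted lasso and rope bones, with a separate holster node.
# Pose Drivers read clavicle and upper-arm rotations from the CopyPose cache to correct shoulder sleeve bones, using separate Grounded and Skydiving pose assets.
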AnimBlueprint /Game/Characters/Player/Male/Medium/Bodies/M_MED_BadassCowboy/Meshes/M_MED_BadassCowboy_Body_AnimBP
target_skeleton=&/M_MED_BadassCowboy_Body_Skeleton
graphs [
 AnimGraph {nodes [
   n1 = Anim Dynamics - Bone: dyn_jacket_side_01_l(Alpha=1, ExternalForce="0,0,0") {
    angularSpringConstant=115
    bAngularSpring=true
    bGravityOverrideInSimSpace=true
    gravityScale=0.35
    lODThreshold=1
    linearSpringConstant=115
    simulationSpace=RootRelative
    windScale=0.5
    alphaBoolBlend {blendOption=Linear}
    boundBone {boneName=dyn_jacket_side_01_l}
    componentLinearAccScale=<0.1 0.1 0.1>
    componentLinearVelScale=<0.1 0.1 0.1>
    physicsBodyDefinitions [{
      sphereCollisionRadius=0
      boundBone {boneName=dyn_jacket_side_01_l}
      boxExtents=<5 20 20>
      constraintSetup {angularTargetAxis=AxisY, twistAxis=AxisY, angularLimitsMax=<7.5 12 0>, angularLimitsMin=<-7.5 -12 -30>, angularTarget=<0 0 0>}
      localJointOffset=<10 0 0>
     }]
   }
   n2 = Anim Dynamics - Bone: dyn_jacket_side_02_l(ComponentPose=@n1, Alpha=1, ExternalForce="0,0,0") {
    angularSpringConstant=115
    bAngularSpring=true
    bGravityOverrideInSimSpace=true
    gravityScale=0.35
    lODThreshold=1
    linearSpringConstant=115
    simulationSpace=RootRelative
    windScale=0.5
    alphaBoolBlend {blendOption=Linear}
    boundBone {boneName=dyn_jacket_side_02_l}
    componentLinearAccScale=<0.1 0.1 0.1>
    componentLinearVelScale=<0.1 0.1 0.1>
    physicsBodyDefinitions [{
      sphereCollisionRadius=0
      boundBone {boneName=dyn_jacket_side_02_l}
      boxExtents=<5 20 20>
      constraintSetup {angularTargetAxis=AxisY, twistAxis=AxisY, angularLimitsMax=<7.5 12 0>, angularLimitsMin=<-7.5 -12 -30>, angularTarget=<0 0 0>}
      localJointOffset=<10 0 0>
     }]
   }
   n3 = Anim Dynamics - Bone: dyn_jacket_side_03_l(ComponentPose=@n2, Alpha=1, ExternalForce="0,0,0") {
    angularSpringConstant=115
    bAngularSpring=true
    bGravityOverrideInSimSpace=true
    gravityScale=0.35
    lODThreshold=1
    linearSpringConstant=115
    simulationSpace=RootRelative
    windScale=0.5
    alphaBoolBlend {blendOption=Linear}
    boundBone {boneName=dyn_jacket_side_03_l}
    componentLinearAccScale=<0.1 0.1 0.1>
    componentLinearVelScale=<0.1 0.1 0.1>
    physicsBodyDefinitions [{
      sphereCollisionRadius=0
      boundBone {boneName=dyn_jacket_side_03_l}
      boxExtents=<5 20 20>
      constraintSetup {angularTargetAxis=AxisY, twistAxis=AxisY, angularLimitsMax=<7.5 12 2>, angularLimitsMin=<-7.5 -12 -30>, angularTarget=<0 0 0>}
      localJointOffset=<10 0 0>
     }]
   }
   n4 = Anim Dynamics - Bone: dyn_jacket_side_04_l(ComponentPose=@n3, Alpha=1, ExternalForce="0,0,0") {
    angularSpringConstant=115
    bAngularSpring=true
    bGravityOverrideInSimSpace=true
    gravityScale=0.35
    lODThreshold=1
    linearSpringConstant=115
    simulationSpace=RootRelative
    windScale=0.5
    alphaBoolBlend {blendOption=Linear}
    boundBone {boneName=dyn_jacket_side_04_l}
    componentLinearAccScale=<0.1 0.1 0.1>
    componentLinearVelScale=<0.1 0.1 0.1>
    physicsBodyDefinitions [{
      sphereCollisionRadius=0
      boundBone {boneName=dyn_jacket_side_04_l}
      boxExtents=<5 20 20>
      constraintSetup {angularTargetAxis=AxisY, twistAxis=AxisY, angularLimitsMax=<7.5 12 2>, angularLimitsMin=<-7.5 -12 -30>, angularTarget=<0 0 0>}
      localJointOffset=<10 0 0>
     }]
   }
   n5 = Anim Dynamics - Bone: dyn_jacket_side_01_r(ComponentPose=@n4, Alpha=1, ExternalForce="0,0,0") {
    angularSpringConstant=115
    bAngularSpring=true
    bGravityOverrideInSimSpace=true
    gravityScale=0.35
    lODThreshold=1
    linearSpringConstant=115
    simulationSpace=RootRelative
    windScale=0.5
    alphaBoolBlend {blendOption=Linear}
    boundBone {boneName=dyn_jacket_side_01_r}
    componentLinearAccScale=<0.1 0.1 0.1>
    componentLinearVelScale=<0.1 0.1 0.1>
    physicsBodyDefinitions [{
      sphereCollisionRadius=0
      boundBone {boneName=dyn_jacket_side_01_r}
      boxExtents=<5 20 20>
      constraintSetup {angularTargetAxis=AxisY, twistAxis=AxisY, angularLimitsMax=<7.5 12 0>, angularLimitsMin=<-7.5 -12 -30>, angularTarget=<0 0 0>}
      localJointOffset=<10 0 0>
     }]
   }
   n6 = Anim Dynamics - Bone: dyn_jacket_side_02_r(ComponentPose=@n5, Alpha=1, ExternalForce="0,0,0") {
    angularSpringConstant=115
    bAngularSpring=true
    bGravityOverrideInSimSpace=true
    gravityScale=0.35
    lODThreshold=1
    linearSpringConstant=115
    simulationSpace=RootRelative
    windScale=0.5
    alphaBoolBlend {blendOption=Linear}
    boundBone {boneName=dyn_jacket_side_02_r}
    componentLinearAccScale=<0.1 0.1 0.1>
    componentLinearVelScale=<0.1 0.1 0.1>
    physicsBodyDefinitions [{
      sphereCollisionRadius=0
      boundBone {boneName=dyn_jacket_side_02_r}
      boxExtents=<5 20 20>
      constraintSetup {angularTargetAxis=AxisY, twistAxis=AxisY, angularLimitsMax=<7.5 12 0>, angularLimitsMin=<-7.5 -12 -30>, angularTarget=<0 0 0>}
      localJointOffset=<10 0 0>
     }]
   }
   n7 = Anim Dynamics - Bone: dyn_jacket_side_03_r(ComponentPose=@n6, Alpha=1, ExternalForce="0,0,0") {
    angularSpringConstant=115
    bAngularSpring=true
    bGravityOverrideInSimSpace=true
    gravityScale=0.35
    lODThreshold=1
    linearSpringConstant=115
    simulationSpace=RootRelative
    windScale=0.5
    alphaBoolBlend {blendOption=Linear}
    boundBone {boneName=dyn_jacket_side_03_r}
    componentLinearAccScale=<0.1 0.1 0.1>
    componentLinearVelScale=<0.1 0.1 0.1>
    physicsBodyDefinitions [{
      sphereCollisionRadius=0
      boundBone {boneName=dyn_jacket_side_03_r}
      boxExtents=<5 20 20>
      constraintSetup {angularTargetAxis=AxisY, twistAxis=AxisY, angularLimitsMax=<7.5 12 2>, angularLimitsMin=<-7.5 -12 -30>, angularTarget=<0 0 0>}
      localJointOffset=<10 0 0>
     }]
   }
   n8 = Anim Dynamics - Bone: dyn_jacket_side_04_r(ComponentPose=@n7, Alpha=1, ExternalForce="0,0,0") {
    angularSpringConstant=115
    bAngularSpring=true
    bGravityOverrideInSimSpace=true
    gravityScale=0.35
    lODThreshold=1
    linearSpringConstant=115
    simulationSpace=RootRelative
    windScale=0.5
    alphaBoolBlend {blendOption=Linear}
    boundBone {boneName=dyn_jacket_side_04_r}
    componentLinearAccScale=<0.1 0.1 0.1>
    componentLinearVelScale=<0.1 0.1 0.1>
    physicsBodyDefinitions [{
      sphereCollisionRadius=0
      boundBone {boneName=dyn_jacket_side_04_r}
      boxExtents=<5 20 20>
      constraintSetup {angularTargetAxis=AxisY, twistAxis=AxisY, angularLimitsMax=<7.5 12 2>, angularLimitsMin=<-7.5 -12 -30>, angularTarget=<0 0 0>}
      localJointOffset=<10 0 0>
     }]
   }
   n9 = Anim Dynamics - Bone: dyn_jacket_front_01_l(ComponentPose=@n8, Alpha=1, ExternalForce="0,0,0") {
    angularSpringConstant=115
    bAngularSpring=true
    bGravityOverrideInSimSpace=true
    gravityScale=0.35
    lODThreshold=1
    linearSpringConstant=115
    simulationSpace=RootRelative
    windScale=0.5
    alphaBoolBlend {blendOption=Linear}
    boundBone {boneName=dyn_jacket_front_01_l}
    componentLinearAccScale=<0.1 0.1 0.1>
    componentLinearVelScale=<0.1 0.1 0.1>
    physicsBodyDefinitions [{
      sphereCollisionRadius=0
      boundBone {boneName=dyn_jacket_front_01_l}
      boxExtents=<5 20 20>
      constraintSetup {angularTargetAxis=AxisY, twistAxis=AxisY, angularLimitsMax=<7.5 12 0>, angularLimitsMin=<-7.5 -12 -30>, angularTarget=<0 0 0>}
      localJointOffset=<10 0 0>
     }]
   }
   n10 = Anim Dynamics - Bone: dyn_jacket_front_02_l(ComponentPose=@n9, Alpha=1, ExternalForce="0,0,0") {
    angularSpringConstant=115
    bAngularSpring=true
    bGravityOverrideInSimSpace=true
    gravityScale=0.35
    lODThreshold=1
    linearSpringConstant=115
    simulationSpace=RootRelative
    windScale=0.5
    alphaBoolBlend {blendOption=Linear}
    boundBone {boneName=dyn_jacket_front_02_l}
    componentLinearAccScale=<0.1 0.1 0.1>
    componentLinearVelScale=<0.1 0.1 0.1>
    physicsBodyDefinitions [{
      sphereCollisionRadius=0
      boundBone {boneName=dyn_jacket_front_02_l}
      boxExtents=<5 20 20>
      constraintSetup {angularTargetAxis=AxisY, twistAxis=AxisY, angularLimitsMax=<7.5 12 0>, angularLimitsMin=<-7.5 -12 -30>, angularTarget=<0 0 0>}
      localJointOffset=<10 0 0>
     }]
   }
   n11 = Anim Dynamics - Bone: dyn_jacket_front_03_l(ComponentPose=@n10, Alpha=1, ExternalForce="0,0,0") {
    angularSpringConstant=115
    bAngularSpring=true
    bGravityOverrideInSimSpace=true
    gravityScale=0.35
    lODThreshold=1
    linearSpringConstant=115
    simulationSpace=RootRelative
    windScale=0.5
    alphaBoolBlend {blendOption=Linear}
    boundBone {boneName=dyn_jacket_front_03_l}
    componentLinearAccScale=<0.1 0.1 0.1>
    componentLinearVelScale=<0.1 0.1 0.1>
    physicsBodyDefinitions [{
      sphereCollisionRadius=0
      boundBone {boneName=dyn_jacket_front_03_l}
      boxExtents=<5 20 20>
      constraintSetup {angularTargetAxis=AxisY, twistAxis=AxisY, angularLimitsMax=<7.5 12 2>, angularLimitsMin=<-7.5 -12 -30>, angularTarget=<0 0 0>}
      localJointOffset=<10 0 0>
     }]
   }
   n12 = Anim Dynamics - Bone: dyn_jacket_front_04_l(ComponentPose=@n11, Alpha=1, ExternalForce="0,0,0") {
    angularSpringConstant=115
    bAngularSpring=true
    bGravityOverrideInSimSpace=true
    gravityScale=0.35
    lODThreshold=1
    linearSpringConstant=115
    simulationSpace=RootRelative
    windScale=0.5
    alphaBoolBlend {blendOption=Linear}
    boundBone {boneName=dyn_jacket_front_04_l}
    componentLinearAccScale=<0.1 0.1 0.1>
    componentLinearVelScale=<0.1 0.1 0.1>
    physicsBodyDefinitions [{
      sphereCollisionRadius=0
      boundBone {boneName=dyn_jacket_front_04_l}
      boxExtents=<5 20 20>
      constraintSetup {angularTargetAxis=AxisY, twistAxis=AxisY, angularLimitsMax=<7.5 12 2>, angularLimitsMin=<-7.5 -12 -30>, angularTarget=<0 0 0>}
      localJointOffset=<10 0 0>
     }]
   }
   n13 = Anim Dynamics - Bone: dyn_jacket_front_01_r(ComponentPose=@n12, Alpha=1, ExternalForce="0,0,0") {
    angularSpringConstant=115
    bAngularSpring=true
    bGravityOverrideInSimSpace=true
    gravityScale=0.35
    lODThreshold=1
    linearSpringConstant=115
    simulationSpace=RootRelative
    windScale=0.5
    alphaBoolBlend {blendOption=Linear}
    boundBone {boneName=dyn_jacket_front_01_r}
    componentLinearAccScale=<0.1 0.1 0.1>
    componentLinearVelScale=<0.1 0.1 0.1>
    physicsBodyDefinitions [{
      sphereCollisionRadius=0
      boundBone {boneName=dyn_jacket_front_01_r}
      boxExtents=<5 20 20>
      constraintSetup {angularTargetAxis=AxisY, twistAxis=AxisY, angularLimitsMax=<7.5 12 0>, angularLimitsMin=<-7.5 -12 -30>, angularTarget=<0 0 0>}
      localJointOffset=<10 0 0>
     }]
   }
   n14 = Anim Dynamics - Bone: dyn_jacket_front_02_r(ComponentPose=@n13, Alpha=1, ExternalForce="0,0,0") {
    angularSpringConstant=115
    bAngularSpring=true
    bGravityOverrideInSimSpace=true
    gravityScale=0.35
    lODThreshold=1
    linearSpringConstant=115
    simulationSpace=RootRelative
    windScale=0.5
    alphaBoolBlend {blendOption=Linear}
    boundBone {boneName=dyn_jacket_front_02_r}
    componentLinearAccScale=<0.1 0.1 0.1>
    componentLinearVelScale=<0.1 0.1 0.1>
    physicsBodyDefinitions [{
      sphereCollisionRadius=0
      boundBone {boneName=dyn_jacket_front_02_r}
      boxExtents=<5 20 20>
      constraintSetup {angularTargetAxis=AxisY, twistAxis=AxisY, angularLimitsMax=<7.5 12 0>, angularLimitsMin=<-7.5 -12 -30>, angularTarget=<0 0 0>}
      localJointOffset=<10 0 0>
     }]
   }
   n15 = Anim Dynamics - Bone: dyn_jacket_front_03_r(ComponentPose=@n14, Alpha=1, ExternalForce="0,0,0") {
    angularSpringConstant=115
    bAngularSpring=true
    bGravityOverrideInSimSpace=true
    gravityScale=0.35
    lODThreshold=1
    linearSpringConstant=115
    simulationSpace=RootRelative
    windScale=0.5
    alphaBoolBlend {blendOption=Linear}
    boundBone {boneName=dyn_jacket_front_03_r}
    componentLinearAccScale=<0.1 0.1 0.1>
    componentLinearVelScale=<0.1 0.1 0.1>
    physicsBodyDefinitions [{
      sphereCollisionRadius=0
      boundBone {boneName=dyn_jacket_front_03_r}
      boxExtents=<5 20 20>
      constraintSetup {angularTargetAxis=AxisY, twistAxis=AxisY, angularLimitsMax=<7.5 12 2>, angularLimitsMin=<-7.5 -12 -30>, angularTarget=<0 0 0>}
      localJointOffset=<10 0 0>
     }]
   }
   n16 = Anim Dynamics - Bone: dyn_jacket_front_04_r(ComponentPose=@n15, Alpha=1, ExternalForce="0,0,0") {
    angularSpringConstant=115
    bAngularSpring=true
    bGravityOverrideInSimSpace=true
    gravityScale=0.35
    lODThreshold=1
    linearSpringConstant=115
    simulationSpace=RootRelative
    windScale=0.5
    alphaBoolBlend {blendOption=Linear}
    boundBone {boneName=dyn_jacket_front_04_r}
    componentLinearAccScale=<0.1 0.1 0.1>
    componentLinearVelScale=<0.1 0.1 0.1>
    physicsBodyDefinitions [{
      sphereCollisionRadius=0
      boundBone {boneName=dyn_jacket_front_04_r}
      boxExtents=<5 20 20>
      constraintSetup {angularTargetAxis=AxisY, twistAxis=AxisY, angularLimitsMax=<7.5 12 2>, angularLimitsMin=<-7.5 -12 -30>, angularTarget=<0 0 0>}
      localJointOffset=<10 0 0>
     }]
   }
   n17 = Anim Dynamics - Bone: dyn_jacket_rear_01_l(ComponentPose=@n16, Alpha=1, ExternalForce="0,0,0") {
    angularSpringConstant=115
    bAngularSpring=true
    bGravityOverrideInSimSpace=true
    gravityScale=0.35
    lODThreshold=1
    linearSpringConstant=115
    simulationSpace=RootRelative
    windScale=0.5
    alphaBoolBlend {blendOption=Linear}
    boundBone {boneName=dyn_jacket_rear_01_l}
    componentLinearAccScale=<0.1 0.1 0.1>
    componentLinearVelScale=<0.1 0.1 0.1>
    physicsBodyDefinitions [{
      sphereCollisionRadius=0
      boundBone {boneName=dyn_jacket_rear_01_l}
      boxExtents=<5 20 20>
      constraintSetup {angularTargetAxis=AxisY, twistAxis=AxisY, angularLimitsMax=<7.5 12 0>, angularLimitsMin=<-7.5 -12 -30>, angularTarget=<0 0 0>}
      localJointOffset=<10 0 0>
     }]
   }
   n18 = Anim Dynamics - Bone: dyn_jacket_rear_02_l(ComponentPose=@n17, Alpha=1, ExternalForce="0,0,0") {
    angularSpringConstant=115
    bAngularSpring=true
    bGravityOverrideInSimSpace=true
    gravityScale=0.35
    lODThreshold=1
    linearSpringConstant=115
    simulationSpace=RootRelative
    windScale=0.5
    alphaBoolBlend {blendOption=Linear}
    boundBone {boneName=dyn_jacket_rear_02_l}
    componentLinearAccScale=<0.1 0.1 0.1>
    componentLinearVelScale=<0.1 0.1 0.1>
    physicsBodyDefinitions [{
      sphereCollisionRadius=0
      boundBone {boneName=dyn_jacket_rear_02_l}
      boxExtents=<5 20 20>
      constraintSetup {angularTargetAxis=AxisY, twistAxis=AxisY, angularLimitsMax=<7.5 12 0>, angularLimitsMin=<-7.5 -12 -30>, angularTarget=<0 0 0>}
      localJointOffset=<10 0 0>
     }]
   }
   n19 = Anim Dynamics - Bone: dyn_jacket_rear_03_l(ComponentPose=@n18, Alpha=1, ExternalForce="0,0,0") {
    angularSpringConstant=115
    bAngularSpring=true
    bGravityOverrideInSimSpace=true
    gravityScale=0.35
    lODThreshold=1
    linearSpringConstant=115
    simulationSpace=RootRelative
    windScale=0.5
    alphaBoolBlend {blendOption=Linear}
    boundBone {boneName=dyn_jacket_rear_03_l}
    componentLinearAccScale=<0.1 0.1 0.1>
    componentLinearVelScale=<0.1 0.1 0.1>
    physicsBodyDefinitions [{
      sphereCollisionRadius=0
      boundBone {boneName=dyn_jacket_rear_03_l}
      boxExtents=<5 20 20>
      constraintSetup {angularTargetAxis=AxisY, twistAxis=AxisY, angularLimitsMax=<7.5 12 2>, angularLimitsMin=<-7.5 -12 -30>, angularTarget=<0 0 0>}
      localJointOffset=<10 0 0>
     }]
   }
   n20 = Anim Dynamics - Bone: dyn_jacket_rear_04_l(ComponentPose=@n19, Alpha=1, ExternalForce="0,0,0") {
    angularSpringConstant=115
    bAngularSpring=true
    bGravityOverrideInSimSpace=true
    gravityScale=0.35
    lODThreshold=1
    linearSpringConstant=115
    simulationSpace=RootRelative
    windScale=0.5
    alphaBoolBlend {blendOption=Linear}
    boundBone {boneName=dyn_jacket_rear_04_l}
    componentLinearAccScale=<0.1 0.1 0.1>
    componentLinearVelScale=<0.1 0.1 0.1>
    physicsBodyDefinitions [{
      sphereCollisionRadius=0
      boundBone {boneName=dyn_jacket_rear_04_l}
      boxExtents=<5 20 20>
      constraintSetup {angularTargetAxis=AxisY, twistAxis=AxisY, angularLimitsMax=<7.5 12 2>, angularLimitsMin=<-7.5 -12 -30>, angularTarget=<0 0 0>}
      localJointOffset=<10 0 0>
     }]
   }
   n21 = Anim Dynamics - Bone: dyn_jacket_rear_01_r(ComponentPose=@n20, Alpha=1, ExternalForce="0,0,0") {
    angularSpringConstant=115
    bAngularSpring=true
    bGravityOverrideInSimSpace=true
    gravityScale=0.35
    lODThreshold=1
    linearSpringConstant=115
    simulationSpace=RootRelative
    windScale=0.5
    alphaBoolBlend {blendOption=Linear}
    boundBone {boneName=dyn_jacket_rear_01_r}
    componentLinearAccScale=<0.1 0.1 0.1>
    componentLinearVelScale=<0.1 0.1 0.1>
    physicsBodyDefinitions [{
      sphereCollisionRadius=0
      boundBone {boneName=dyn_jacket_rear_01_r}
      boxExtents=<5 20 20>
      constraintSetup {angularTargetAxis=AxisY, twistAxis=AxisY, angularLimitsMax=<7.5 12 0>, angularLimitsMin=<-7.5 -12 -30>, angularTarget=<0 0 0>}
      localJointOffset=<10 0 0>
     }]
   }
   n22 = Anim Dynamics - Bone: dyn_jacket_rear_02_r(ComponentPose=@n21, Alpha=1, ExternalForce="0,0,0") {
    angularSpringConstant=115
    bAngularSpring=true
    bGravityOverrideInSimSpace=true
    gravityScale=0.35
    lODThreshold=1
    linearSpringConstant=115
    simulationSpace=RootRelative
    windScale=0.5
    alphaBoolBlend {blendOption=Linear}
    boundBone {boneName=dyn_jacket_rear_02_r}
    componentLinearAccScale=<0.1 0.1 0.1>
    componentLinearVelScale=<0.1 0.1 0.1>
    physicsBodyDefinitions [{
      sphereCollisionRadius=0
      boundBone {boneName=dyn_jacket_rear_02_r}
      boxExtents=<5 20 20>
      constraintSetup {angularTargetAxis=AxisY, twistAxis=AxisY, angularLimitsMax=<7.5 12 0>, angularLimitsMin=<-7.5 -12 -30>, angularTarget=<0 0 0>}
      localJointOffset=<10 0 0>
     }]
   }
   n23 = Anim Dynamics - Bone: dyn_jacket_rear_03_r(ComponentPose=@n22, Alpha=1, ExternalForce="0,0,0") {
    angularSpringConstant=115
    bAngularSpring=true
    bGravityOverrideInSimSpace=true
    gravityScale=0.35
    lODThreshold=1
    linearSpringConstant=115
    simulationSpace=RootRelative
    windScale=0.5
    alphaBoolBlend {blendOption=Linear}
    boundBone {boneName=dyn_jacket_rear_03_r}
    componentLinearAccScale=<0.1 0.1 0.1>
    componentLinearVelScale=<0.1 0.1 0.1>
    physicsBodyDefinitions [{
      sphereCollisionRadius=0
      boundBone {boneName=dyn_jacket_rear_03_r}
      boxExtents=<5 20 20>
      constraintSetup {angularTargetAxis=AxisY, twistAxis=AxisY, angularLimitsMax=<7.5 12 2>, angularLimitsMin=<-7.5 -12 -30>, angularTarget=<0 0 0>}
      localJointOffset=<10 0 0>
     }]
   }
   Anim Dynamics - Bone: dyn_jacket_rear_04_r(ComponentPose=@n23, Alpha=1, ExternalForce="0,0,0") {
    angularSpringConstant=115
    bAngularSpring=true
    bGravityOverrideInSimSpace=true
    gravityScale=0.35
    lODThreshold=1
    linearSpringConstant=115
    simulationSpace=RootRelative
    windScale=0.5
    alphaBoolBlend {blendOption=Linear}
    boundBone {boneName=dyn_jacket_rear_04_r}
    componentLinearAccScale=<0.1 0.1 0.1>
    componentLinearVelScale=<0.1 0.1 0.1>
    physicsBodyDefinitions [{
      sphereCollisionRadius=0
      boundBone {boneName=dyn_jacket_rear_04_r}
      boxExtents=<5 20 20>
      constraintSetup {angularTargetAxis=AxisY, twistAxis=AxisY, angularLimitsMax=<7.5 12 2>, angularLimitsMin=<-7.5 -12 -30>, angularTarget=<0 0 0>}
      localJointOffset=<10 0 0>
     }]
   }
   n24 = Anim Dynamics - Bone: dyn_lasso_01(Alpha=0.3, ExternalForce="0,0,0") {
    angularDampingOverride=0.99
    angularSpringConstant=250
    bAngularSpring=true
    bGravityOverrideInSimSpace=true
    bOverrideAngularDamping=true
    gravityScale=0.5
    lODThreshold=1
    simulationSpace=RootRelative
    windScale=0.5
    alphaBoolBlend {blendOption=Linear}
    boundBone {boneName=dyn_lasso_01}
    componentAppliedLinearAccClamp=<1000 1000 1000>
    componentLinearAccScale=<1 1 1>
    componentLinearVelScale=<1 1 1>
    physicsBodyDefinitions [{
      sphereCollisionRadius=0
      boundBone {boneName=dyn_lasso_01}
      boxExtents=<10 20 20>
      constraintSetup {angularLimitsMax=<0 30 30>, angularLimitsMin=<0 -30 0>, angularTarget=<0 0 0>}
      localJointOffset=<10 0 0>
     }]
   }
   n25 = Anim Dynamics - Bone: dyn_lasso_02(ComponentPose=@n24, Alpha=0.3, ExternalForce="0,0,0") {
    angularDampingOverride=0.99
    angularSpringConstant=250
    bAngularSpring=true
    bGravityOverrideInSimSpace=true
    bOverrideAngularDamping=true
    gravityScale=0.5
    lODThreshold=1
    windScale=0.5
    alphaBoolBlend {blendOption=Linear}
    boundBone {boneName=dyn_lasso_02}
    componentAppliedLinearAccClamp=<1000 1000 1000>
    componentLinearAccScale=<1 1 1>
    componentLinearVelScale=<1 1 1>
    physicsBodyDefinitions [{
      sphereCollisionRadius=0
      boundBone {boneName=dyn_lasso_02}
      boxExtents=<5 20 20>
      constraintSetup {angularLimitsMax=<15 20 60>, angularLimitsMin=<-15 -20 -5>, angularTarget=<0 0 0>}
      localJointOffset=<1 0 0>
     }]
   }
   n26 = Anim Dynamics - Bone: dyn_lasso_03(ComponentPose=@n25, Alpha=0.3, ExternalForce="0,0,0") {
    angularDampingOverride=0.99
    angularSpringConstant=250
    bAngularSpring=true
    bGravityOverrideInSimSpace=true
    bOverrideAngularDamping=true
    gravityScale=0.5
    lODThreshold=1
    windScale=0.5
    alphaBoolBlend {blendOption=Linear}
    boundBone {boneName=dyn_lasso_03}
    componentAppliedLinearAccClamp=<1000 1000 1000>
    componentLinearAccScale=<1 1 1>
    componentLinearVelScale=<1 1 1>
    physicsBodyDefinitions [{
      sphereCollisionRadius=0
      boundBone {boneName=dyn_lasso_03}
      boxExtents=<5 20 20>
      constraintSetup {angularLimitsMax=<15 20 60>, angularLimitsMin=<-15 -20 -5>, angularTarget=<0 0 0>}
      localJointOffset=<1 0 0>
     }]
   }
   n27 = Anim Dynamics - Bone: dyn_rope_01(ComponentPose=@n26, Alpha=0.2, ExternalForce="0,0,0") {
    angularDampingOverride=0.99
    angularSpringConstant=250
    bAngularSpring=true
    bGravityOverrideInSimSpace=true
    bOverrideAngularDamping=true
    lODThreshold=1
    windScale=0.5
    alphaBoolBlend {blendOption=Linear}
    boundBone {boneName=dyn_rope_01}
    componentAppliedLinearAccClamp=<1000 1000 1000>
    componentLinearAccScale=<1 1 1>
    componentLinearVelScale=<1 1 1>
    physicsBodyDefinitions [{
      sphereCollisionRadius=0
      boundBone {boneName=dyn_rope_01}
      boxExtents=<10 20 20>
      constraintSetup {angularLimitsMax=<0 30 30>, angularLimitsMin=<0 -30 0>, angularTarget=<0 0 0>}
      localJointOffset=<1 0 0>
     }]
   }
   Anim Dynamics - Bone: dyn_rope_02(ComponentPose=@n27, Alpha=0.2, ExternalForce="0,0,0") {
    angularDampingOverride=0.99
    angularSpringConstant=250
    bAngularSpring=true
    bGravityOverrideInSimSpace=true
    bOverrideAngularDamping=true
    lODThreshold=1
    windScale=0.5
    alphaBoolBlend {blendOption=Linear}
    boundBone {boneName=dyn_rope_02}
    componentAppliedLinearAccClamp=<1000 1000 1000>
    componentLinearAccScale=<1 1 1>
    componentLinearVelScale=<1 1 1>
    physicsBodyDefinitions [{
      sphereCollisionRadius=0
      boundBone {boneName=dyn_rope_02}
      boxExtents=<10 20 20>
      constraintSetup {angularLimitsMax=<0 30 30>, angularLimitsMin=<0 -30 0>, angularTarget=<0 0 0>}
      localJointOffset=<1 0 0>
     }]
   }
   Anim Dynamics - Bone: dyn_holster(Alpha=1, ExternalForce="0,0,0") {
    angularDampingOverride=0.999
    angularSpringConstant=100
    bAngularSpring=true
    bGravityOverrideInSimSpace=true
    bOverrideAngularDamping=true
    bUseGravityOverride=true
    gravityScale=0.35
    simulationSpace=RootRelative
    alphaBoolBlend {blendOption=Linear}
    boundBone {boneName=dyn_holster}
    componentAppliedLinearAccClamp=<3000 3000 3000>
    componentLinearAccScale=<0.5 0.5 0.5>
    componentLinearVelScale=<0.5 0.5 0.5>
    gravityOverride {z=-980}
    physicsBodyDefinitions [{
      sphereCollisionRadius=0
      boundBone {boneName=dyn_holster}
      boxExtents=<10 30 30>
      constraintSetup {angularLimitsMax=<0 0 5>, angularLimitsMin=<0 0 -5>, angularTarget=<0 0 0>}
      localJointOffset=<5 0 0>
     }]
   }
   n28 = Pose Driver - Source: clavicle_l!(SourcePose:Pose Link Structure=Use cached pose 'CopyPose_Cache'()) -> (Pose:Pose Link Structure) {
    is=title AnimGraphNode_PoseDriver
    poseAsset=&/M_MED_Badass_Cowboy_Grounded_PoseAsset
    onlyDriveBones [{boneName=dyn_shoulder_sleeve_01_l}]
    poseTargets [
     {
      bApplyCustomCurve=false
      bIsHidden=false
      distanceMethod=DefaultMethod
      drivenName=Pose_0
      functionType=DefaultFunction
      targetScale=0.0005
      boneTransforms [
       {targetRotation=<rot -85.1915 -138.5625 143.0619>, targetTranslation=<4.3575 -0.1466 -2.3223>}
       {targetRotation=<rot -64.7854 -3.089 6.0959>, targetTranslation=<15.5111 0 0>}
      ]
      customCurve {defaultValue=3.4028234663852886e+38, postInfinityExtrap=RCCE_Constant, preInfinityExtrap=RCCE_Constant, keys []}
      targetRotation=<rot 0 0 0>
     }
     {
      bApplyCustomCurve=false
      bIsHidden=false
      distanceMethod=DefaultMethod
      drivenName=Pose_1
      functionType=DefaultFunction
      targetScale=1
      boneTransforms [
       {targetRotation=<rot -71.8075 -7.0773 -33.3059>, targetTranslation=<4.3575 -0.1466 -2.3223>}
       {targetRotation=<rot -9.5233 -21.9053 -28.615>, targetTranslation=<15.5111 0 0>}
      ]
      customCurve {defaultValue=3.4028234663852886e+38, postInfinityExtrap=RCCE_Constant, preInfinityExtrap=RCCE_Constant, keys []}
      targetRotation=<rot 0 0 0>
     }
     {
      bApplyCustomCurve=false
      bIsHidden=false
      distanceMethod=DefaultMethod
      drivenName=Pose_2
      functionType=DefaultFunction
      targetScale=0.308
      boneTransforms [
       {targetRotation=<rot -75.4617 19.5667 -12.8774>, targetTranslation=<4.3575 -0.1466 -2.3223>}
       {targetRotation=<rot -64.9089 79.0369 -47.2079>, targetTranslation=<15.5111 0 0>}
      ]
      customCurve {defaultValue=3.4028234663852886e+38, postInfinityExtrap=RCCE_Constant, preInfinityExtrap=RCCE_Constant, keys []}
      targetRotation=<rot 0 0 0>
     }
     {
      bApplyCustomCurve=false
      bIsHidden=false
      distanceMethod=DefaultMethod
      drivenName=Pose_3
      functionType=DefaultFunction
      targetScale=0.4931
      boneTransforms [
       {targetRotation=<rot -74.7073 55.0349 -49.5566>, targetTranslation=<4.3575 -0.1466 -2.3223>}
       {targetRotation=<rot -36.5188 -9.2495 27.844>, targetTranslation=<15.5111 0 0>}
      ]
      customCurve {defaultValue=3.4028234663852886e+38, postInfinityExtrap=RCCE_Constant, preInfinityExtrap=RCCE_Constant, keys []}
      targetRotation=<rot 0 0 0>
     }
     {
      bApplyCustomCurve=false
      bIsHidden=false
      distanceMethod=DefaultMethod
      drivenName=Pose_4
      functionType=DefaultFunction
      targetScale=0.2379
      boneTransforms [
       {targetRotation=<rot -81.7412 -135.4938 140.2621>, targetTranslation=<4.3575 -0.1466 -2.3223>}
       {targetRotation=<rot -49.8941 -8.1717 1.1726>, targetTranslation=<15.5111 0 0>}
      ]
      customCurve {defaultValue=3.4028234663852886e+38, postInfinityExtrap=RCCE_Constant, preInfinityExtrap=RCCE_Constant, keys []}
      targetRotation=<rot 0 0 0>
     }
     {
      bApplyCustomCurve=false
      bIsHidden=false
      distanceMethod=DefaultMethod
      drivenName=Pose_5
      functionType=DefaultFunction
      targetScale=0.7376
      boneTransforms [
       {targetRotation=<rot -80.8622 39.0398 -38.1227>, targetTranslation=<4.3575 -0.1466 -2.3223>}
       {targetRotation=<rot -31.8256 65.8251 -86.0604>, targetTranslation=<15.5111 0 0>}
      ]
      customCurve {defaultValue=3.4028234663852886e+38, postInfinityExtrap=RCCE_Constant, preInfinityExtrap=RCCE_Constant, keys []}
      targetRotation=<rot 0 0 0>
     }
     {
      bApplyCustomCurve=false
      bIsHidden=false
      distanceMethod=DefaultMethod
      drivenName=Pose_6
      functionType=DefaultFunction
      targetScale=0.558
      boneTransforms [
       {targetRotation=<rot -47.5697 14.692 -0.434>, targetTranslation=<4.3575 -0.1466 -2.3223>}
       {targetRotation=<rot -8.3098 26.4205 -78.9654>, targetTranslation=<15.5111 0 0>}
      ]
      customCurve {defaultValue=3.4028234663852886e+38, postInfinityExtrap=RCCE_Constant, preInfinityExtrap=RCCE_Constant, keys []}
      targetRotation=<rot 0 0 0>
     }
     {
      bApplyCustomCurve=false
      bIsHidden=false
      distanceMethod=DefaultMethod
      drivenName=Pose_7
      functionType=DefaultFunction
      targetScale=0.308
      boneTransforms [
       {targetRotation=<rot -86.0332 143.3968 -137.767>, targetTranslation=<4.3575 -0.1466 -2.3223>}
       {targetRotation=<rot -67.1606 51.1714 -32.7303>, targetTranslation=<15.5111 0 0>}
      ]
      customCurve {defaultValue=3.4028234663852886e+38, postInfinityExtrap=RCCE_Constant, preInfinityExtrap=RCCE_Constant, keys []}
      targetRotation=<rot 0 0 0>
     }
     {
      bApplyCustomCurve=false
      bIsHidden=false
      distanceMethod=DefaultMethod
      drivenName=Pose_8
      functionType=DefaultFunction
      targetScale=0.0005
      boneTransforms [
       {targetRotation=<rot -85.1915 -138.5625 143.0619>, targetTranslation=<4.3575 -0.1466 -2.3223>}
       {targetRotation=<rot -64.7854 -3.089 6.0959>, targetTranslation=<15.5111 0 0>}
      ]
      customCurve {defaultValue=3.4028234663852886e+38, postInfinityExtrap=RCCE_Constant, preInfinityExtrap=RCCE_Constant, keys []}
      targetRotation=<rot 0 0 0>
     }
     {
      bApplyCustomCurve=false
      bIsHidden=false
      distanceMethod=DefaultMethod
      drivenName=Pose_9
      functionType=DefaultFunction
      targetScale=0.558
      boneTransforms [
       {targetRotation=<rot -48.356 10.1188 -2.4428>, targetTranslation=<4.3575 -0.1466 -2.3223>}
       {targetRotation=<rot 14.5763 54.8781 -98.5188>, targetTranslation=<15.5111 0 0>}
      ]
      customCurve {defaultValue=3.4028234663852886e+38, postInfinityExtrap=RCCE_Constant, preInfinityExtrap=RCCE_Constant, keys []}
      targetRotation=<rot 0 0 0>
     }
     {
      bApplyCustomCurve=false
      bIsHidden=false
      distanceMethod=DefaultMethod
      drivenName=Pose_10
      functionType=DefaultFunction
      targetScale=0.7228
      boneTransforms [
       {targetRotation=<rot -79.2874 99.6397 -110.9321>, targetTranslation=<4.3575 -0.1466 -2.3223>}
       {targetRotation=<rot -14.9047 102.8909 -17.6919>, targetTranslation=<15.5111 0 0>}
      ]
      customCurve {defaultValue=3.4028234663852886e+38, postInfinityExtrap=RCCE_Constant, preInfinityExtrap=RCCE_Constant, keys []}
      targetRotation=<rot 0 0 0>
     }
     {
      bApplyCustomCurve=false
      bIsHidden=false
      distanceMethod=DefaultMethod
      drivenName=Pose_11
      functionType=DefaultFunction
      targetScale=0.7228
      boneTransforms [
       {targetRotation=<rot -58.4846 -9.5373 2.6759>, targetTranslation=<4.3575 -0.1466 -2.3223>}
       {targetRotation=<rot -44.7022 128.4906 -48.0679>, targetTranslation=<15.5111 0 0>}
      ]
      customCurve {defaultValue=3.4028234663852886e+38, postInfinityExtrap=RCCE_Constant, preInfinityExtrap=RCCE_Constant, keys []}
      targetRotation=<rot 0 0 0>
     }
    ]
    rBFParams {distanceMethod=Quaternion, function=Linear, radius=37.9404}
    sourceBones [{boneName=clavicle_l}, {boneName=upperarm_l}]
   }
   n29 = Pose Driver - Source: clavicle_r!(SourcePose:Pose Link Structure=@n28) -> (Pose:Pose Link Structure) {
    is=title AnimGraphNode_PoseDriver
    poseAsset=&M_MED_Badass_Cowboy_Grounded_PoseAsset
    onlyDriveBones [{boneName=dyn_shoulder_sleeve_01_r}]
    poseTargets [
     {
      bApplyCustomCurve=false
      bIsHidden=false
      distanceMethod=DefaultMethod
      drivenName=Pose_0
      functionType=DefaultFunction
      targetScale=0.0007
      boneTransforms [
       {targetRotation=<rot -85.1915 41.4375 143.0619>, targetTranslation=<4.3575 -0.1466 2.3223>}
       {targetRotation=<rot -64.7854 -3.089 6.0959>, targetTranslation=<-15.5111 0 0>}
      ]
      customCurve {defaultValue=3.4028234663852886e+38, postInfinityExtrap=RCCE_Constant, preInfinityExtrap=RCCE_Constant, keys []}
      targetRotation=<rot 0 0 0>
     }
     {
      bApplyCustomCurve=false
      bIsHidden=false
      distanceMethod=DefaultMethod
      drivenName=Pose_1
      functionType=DefaultFunction
      targetScale=0.5605
      boneTransforms [
       {targetRotation=<rot -64.2747 -120.6598 -83.241>, targetTranslation=<4.3575 -0.1466 2.3223>}
       {targetRotation=<rot -23.3709 86.7554 -36.3331>, targetTranslation=<-15.5111 0 0>}
      ]
      customCurve {defaultValue=3.4028234663852886e+38, postInfinityExtrap=RCCE_Constant, preInfinityExtrap=RCCE_Constant, keys []}
      targetRotation=<rot 0 0 0>
     }
     {
      bApplyCustomCurve=false
      bIsHidden=false
      distanceMethod=DefaultMethod
      drivenName=Pose_2
      functionType=DefaultFunction
      targetScale=0.6181
      boneTransforms [
       {targetRotation=<rot -69.838 165.5965 22.6769>, targetTranslation=<4.3575 -0.1466 2.3223>}
       {targetRotation=<rot -75.1088 43.0245 -78.222>, targetTranslation=<-15.5111 0 0>}
      ]
      customCurve {defaultValue=3.4028234663852886e+38, postInfinityExtrap=RCCE_Constant, preInfinityExtrap=RCCE_Constant, keys []}
      targetRotation=<rot 0 0 0>
     }
     {
      bApplyCustomCurve=false
      bIsHidden=false
      distanceMethod=DefaultMethod
      drivenName=Pose_3
      functionType=DefaultFunction
      targetScale=0.3174
      boneTransforms [
       {targetRotation=<rot -73.4134 -176.9703 4.3592>, targetTranslation=<4.3575 -0.1466 2.3223>}
       {targetRotation=<rot -75.5657 105.7208 -93.8077>, targetTranslation=<-15.5111 0 0>}
      ]
      customCurve {defaultValue=3.4028234663852886e+38, postInfinityExtrap=RCCE_Constant, preInfinityExtrap=RCCE_Constant, keys []}
      targetRotation=<rot 0 0 0>
     }
     {
      bApplyCustomCurve=false
      bIsHidden=false
      distanceMethod=DefaultMethod
      drivenName=Pose_4
      functionType=DefaultFunction
      targetScale=0.2938
      boneTransforms [
       {targetRotation=<rot -87.7995 76.2683 109.0762>, targetTranslation=<4.3575 -0.1466 2.3223>}
       {targetRotation=<rot -69.7589 51.1037 -35.2693>, targetTranslation=<-15.5111 0 0>}
      ]
      customCurve {defaultValue=3.4028234663852886e+38, postInfinityExtrap=RCCE_Constant, preInfinityExtrap=RCCE_Constant, keys []}
      targetRotation=<rot 0 0 0>
     }
     {
      bApplyCustomCurve=false
      bIsHidden=false
      distanceMethod=DefaultMethod
      drivenName=Pose_5
      functionType=DefaultFunction
      targetScale=0.2327
      boneTransforms [
       {targetRotation=<rot -76.2754 172.1506 9.5506>, targetTranslation=<4.3575 -0.1466 2.3223>}
       {targetRotation=<rot -45.0185 104.3596 -56.1379>, targetTranslation=<-15.5111 0 0>}
      ]
      customCurve {defaultValue=3.4028234663852886e+38, postInfinityExtrap=RCCE_Constant, preInfinityExtrap=RCCE_Constant, keys []}
      targetRotation=<rot 0 0 0>
     }
     {
      bApplyCustomCurve=false
      bIsHidden=false
      distanceMethod=DefaultMethod
      drivenName=Pose_6
      functionType=DefaultFunction
      targetScale=0.2327
      boneTransforms [
       {targetRotation=<rot -74.8686 -132.7057 -46.977>, targetTranslation=<4.3575 -0.1466 2.3223>}
       {targetRotation=<rot -39.6825 113.2505 -70.71>, targetTranslation=<-15.5111 0 0>}
      ]
      customCurve {defaultValue=3.4028234663852886e+38, postInfinityExtrap=RCCE_Constant, preInfinityExtrap=RCCE_Constant, keys []}
      targetRotation=<rot 0 0 0>
     }
     {
      bApplyCustomCurve=false
      bIsHidden=false
      distanceMethod=DefaultMethod
      drivenName=Pose_7
      functionType=DefaultFunction
      targetScale=0.4039
      boneTransforms [
       {targetRotation=<rot -57.8419 -161.5813 -18.7947>, targetTranslation=<4.3575 -0.1466 2.3223>}
       {targetRotation=<rot -66.8684 120.4314 -71.1722>, targetTranslation=<-15.5111 0 0>}
      ]
      customCurve {defaultValue=3.4028234663852886e+38, postInfinityExtrap=RCCE_Constant, preInfinityExtrap=RCCE_Constant, keys []}
      targetRotation=<rot 0 0 0>
     }
     {
      bApplyCustomCurve=false
      bIsHidden=false
      distanceMethod=DefaultMethod
      drivenName=Pose_8
      functionType=DefaultFunction
      targetScale=0.0007
      boneTransforms [
       {targetRotation=<rot -85.1915 41.4375 143.0619>, targetTranslation=<4.3575 -0.1466 2.3223>}
       {targetRotation=<rot -64.7854 -3.089 6.0959>, targetTranslation=<-15.5111 0 0>}
      ]
      customCurve {defaultValue=3.4028234663852886e+38, postInfinityExtrap=RCCE_Constant, preInfinityExtrap=RCCE_Constant, keys []}
      targetRotation=<rot 0 0 0>
     }
     {
      bApplyCustomCurve=false
      bIsHidden=false
      distanceMethod=DefaultMethod
      drivenName=Pose_9
      functionType=DefaultFunction
      targetScale=1
      boneTransforms [
       {targetRotation=<rot -68.9933 -127.6298 -53.9107>, targetTranslation=<4.3575 -0.1466 2.3223>}
       {targetRotation=<rot 33.2732 89.3676 -89.1674>, targetTranslation=<-15.5111 0 0>}
      ]
      customCurve {defaultValue=3.4028234663852886e+38, postInfinityExtrap=RCCE_Constant, preInfinityExtrap=RCCE_Constant, keys []}
      targetRotation=<rot 0 0 0>
     }
     {
      bApplyCustomCurve=false
      bIsHidden=false
      distanceMethod=DefaultMethod
      drivenName=Pose_10
      functionType=DefaultFunction
      targetScale=0.8736
      boneTransforms [
       {targetRotation=<rot -83.6219 69.4346 115.2169>, targetTranslation=<4.3575 -0.1466 2.3223>}
       {targetRotation=<rot -11.4935 30.1498 -76.7725>, targetTranslation=<-15.5111 0 0>}
      ]
      customCurve {defaultValue=3.4028234663852886e+38, postInfinityExtrap=RCCE_Constant, preInfinityExtrap=RCCE_Constant, keys []}
      targetRotation=<rot 0 0 0>
     }
     {
      bApplyCustomCurve=false
      bIsHidden=false
      distanceMethod=DefaultMethod
      drivenName=Pose_11
      functionType=DefaultFunction
      targetScale=0.571
      boneTransforms [
       {targetRotation=<rot -47.9536 -145.52 -31.796>, targetTranslation=<4.3575 -0.1466 2.3223>}
       {targetRotation=<rot -20.6591 112.0123 -37.7437>, targetTranslation=<-15.5111 0 0>}
      ]
      customCurve {defaultValue=3.4028234663852886e+38, postInfinityExtrap=RCCE_Constant, preInfinityExtrap=RCCE_Constant, keys []}
      targetRotation=<rot 0 0 0>
     }
    ]
    rBFParams {distanceMethod=Quaternion, radius=39.0269}
    sourceBones [{boneName=clavicle_r}, {boneName=upperarm_r}]
   }
   n30 = Pose Driver - Source: upperarm_l!(SourcePose:Pose Link Structure=Use cached pose 'CopyPose_Cache'()) -> (Pose:Pose Link Structure) {
    is=title AnimGraphNode_PoseDriver
    poseAsset=&/M_MED_Badass_Cowboy_Skydiving_PoseAsset
    onlyDriveBones [{boneName=dyn_shoulder_sleeve_01_l}]
    poseTargets [
     {
      bApplyCustomCurve=false
      bIsHidden=false
      distanceMethod=DefaultMethod
      drivenName=Pose_12
      functionType=DefaultFunction
      targetScale=1
      boneTransforms [{targetRotation=<rot -64.7854 -3.089 6.0959>, targetTranslation=<15.5111 0 0>}]
      customCurve {defaultValue=3.4028234663852886e+38, postInfinityExtrap=RCCE_Constant, preInfinityExtrap=RCCE_Constant, keys []}
      targetRotation=<rot 0 0 0>
     }
     {
      bApplyCustomCurve=false
      bIsHidden=false
      distanceMethod=DefaultMethod
      drivenName=Pose_13
      functionType=DefaultFunction
      targetScale=0.3618
      boneTransforms [{targetRotation=<rot -17.9906 -27.503 -67.4976>, targetTranslation=<15.5111 0 0>}]
      customCurve {defaultValue=3.4028234663852886e+38, postInfinityExtrap=RCCE_Constant, preInfinityExtrap=RCCE_Constant, keys []}
      targetRotation=<rot 0 0 0>
     }
     {
      bApplyCustomCurve=false
      bIsHidden=false
      distanceMethod=DefaultMethod
      drivenName=Pose_14
      functionType=DefaultFunction
      targetScale=0.3618
      boneTransforms [{targetRotation=<rot 17.157 -17.077 -68.3926>, targetTranslation=<15.5111 0 0>}]
      customCurve {defaultValue=3.4028234663852886e+38, postInfinityExtrap=RCCE_Constant, preInfinityExtrap=RCCE_Constant, keys []}
      targetRotation=<rot 0 0 0>
     }
    ]
    rBFParams {distanceMethod=Quaternion, radius=50.612}
    sourceBones [{boneName=upperarm_l}]
   }
   n31 = Pose Driver - Source: upperarm_r!(SourcePose:Pose Link Structure=@n30) -> (Pose:Pose Link Structure) {
    is=title AnimGraphNode_PoseDriver
    poseAsset=&M_MED_Badass_Cowboy_Skydiving_PoseAsset
    onlyDriveBones [{boneName=dyn_shoulder_sleeve_01_r}]
    poseTargets [
     {
      bApplyCustomCurve=false
      bIsHidden=false
      distanceMethod=DefaultMethod
      drivenName=Pose_12
      functionType=DefaultFunction
      targetScale=1
      boneTransforms [{targetRotation=<rot -64.7854 -3.089 6.0959>, targetTranslation=<-15.5111 0 0>}]
      customCurve {defaultValue=3.4028234663852886e+38, postInfinityExtrap=RCCE_Constant, preInfinityExtrap=RCCE_Constant, keys []}
      targetRotation=<rot 0 0 0>
     }
     {
      bApplyCustomCurve=false
      bIsHidden=false
      distanceMethod=DefaultMethod
      drivenName=Pose_13
      functionType=DefaultFunction
      targetScale=0.3887
      boneTransforms [{targetRotation=<rot -22.7967 -31.6166 -51.4767>, targetTranslation=<-15.5111 0 0>}]
      customCurve {defaultValue=3.4028234663852886e+38, postInfinityExtrap=RCCE_Constant, preInfinityExtrap=RCCE_Constant, keys []}
      targetRotation=<rot 0 0 0>
     }
     {
      bApplyCustomCurve=false
      bIsHidden=false
      distanceMethod=DefaultMethod
      drivenName=Pose_14
      functionType=DefaultFunction
      targetScale=0.3887
      boneTransforms [{targetRotation=<rot 1.7785 -12.1856 -70.9632>, targetTranslation=<-15.5111 0 0>}]
      customCurve {defaultValue=3.4028234663852886e+38, postInfinityExtrap=RCCE_Constant, preInfinityExtrap=RCCE_Constant, keys []}
      targetRotation=<rot 0 0 0>
     }
    ]
    rBFParams {distanceMethod=Quaternion, radius=44.6911}
    sourceBones [{boneName=upperarm_r}]
   }
   Blend Poses by bool(BlendPose_0=@n31, BlendPose_1=@n29, bActiveValue=Get bIsSkyDiving()) {blendType=Linear}
   "Constraint - dyn_holster_MSTR (PP)"!(ComponentPose, Alpha=1, ConstraintWeights_0=0.6, ConstraintWeights_1:float=0.4) -> (Pose) {
    alphaBoolBlend {blendOption=Linear}
    boneToModify {boneName=dyn_holster_MSTR}
    constraintSetup [{targetBone {boneName=pelvis}}, {targetBone {boneName=thigh_twist_01_l}}]
    constraintWeights [1,1]
   }
   "Constraint - dyn_holster_TGT (PPPPPPPP)"!(
    ComponentPose
    Alpha=1
    ConstraintWeights_0=0.1114
    ConstraintWeights_1:float=0.0078
    ConstraintWeights_2:float=0.3084
    ConstraintWeights_3:float=0.0108
    ConstraintWeights_4:float=0.2984
    ConstraintWeights_5:float=0.0051
    ConstraintWeights_6:float=0.2571
    ConstraintWeights_7:float=0.001
   ) -> (Pose) {
    alphaBoolBlend {blendOption=Linear}
    boneToModify {boneName=dyn_holster_TGT}
    constraintSetup [
     {targetBone {boneName=dyn_jacket_rear_02_l}}
     {targetBone {boneName=dyn_jacket_rear_01_l}}
     {targetBone {boneName=dyn_jacket_rear_03_l}}
     {targetBone {boneName=dyn_jacket_side_01_l}}
     {targetBone {boneName=dyn_jacket_side_03_l}}
     {targetBone {boneName=dyn_jacket_rear_04_l}}
     {targetBone {boneName=dyn_jacket_side_02_l}}
     {targetBone {boneName=pelvis}}
    ]
    constraintWeights [1,1,1,1,1,1,1,1]
   }
   "Constraint - dyn_rope_03 (PP)"!(ComponentPose, Alpha=1, ConstraintWeights_0=0, ConstraintWeights_1:float=1) -> (Pose) {
    alphaBoolBlend {blendOption=Linear}
    boneToModify {boneName=dyn_rope_03}
    constraintSetup [{targetBone {boneName=dyn_jacket_side_03_r}}, {targetBone {boneName=dyn_jacket_rear_03_r}}]
    constraintWeights [1,1]
   }
   n32 = "Constraint - dyn_rope_01_B (PP)"!(ComponentPose, Alpha=1, ConstraintWeights_0=0, ConstraintWeights_1:float=1) -> (Pose) {
    alphaBoolBlend {blendOption=Linear}
    boneToModify {boneName=dyn_rope_01_B}
    constraintSetup [{targetBone {boneName=dyn_jacket_side_01_r}}, {targetBone {boneName=dyn_jacket_rear_01_r}}]
    constraintWeights [1,1]
   }
   "Constraint - dyn_rope_02 (PP)"!(ComponentPose=@n32, Alpha=1, ConstraintWeights_0=0, ConstraintWeights_1:float=1) -> (Pose) {
    alphaBoolBlend {blendOption=Linear}
    boneToModify {boneName=dyn_rope_02}
    constraintSetup [{targetBone {boneName=dyn_jacket_side_02_r}}, {targetBone {boneName=dyn_jacket_rear_02_r}}]
    constraintWeights [1,1]
   }
   "Constraint - dyn_pouch_r (P)"(Alpha=1, ConstraintWeights_0=1) {
    alphaBoolBlend {blendOption=Linear}
    boneToModify {boneName=dyn_pouch_r}
    constraintSetup [{targetBone {boneName=dyn_jacket_side_01_r}}]
    constraintWeights [1]
   }
   n33 = Break Fort Anim Input Skydiving External Force(FortAnimInput_SkydivingExternalForce=Get SkydivingExternalForceInput()) -> (
    bUseSkydivingVectorForce=~
    SkydivingVectorForce=~
    HeadToPelvisDirection=~
    FloatingMultiplier=~
    FloatingAdditive=~
    DivingMultiplier=~
    DivingAdditive=~
    ParachutingMultiplier=~
    ParachutingAdditive=~
   )
   n34 = Get GravityOverride()
   RigidBody!(ComponentPose, Alpha, OverrideWorldGravity:Vector=@n34, ExternalForce=@n33.SkydivingVectorForce) -> (Pose) {
    bClampLinearTranslationLimitToRefPose=true
    bOverrideWorldGravity=true
    overridePhysicsAsset=&/M_MED_BadassCowboy_Body_Perf_Physics
    simulationSpace=BaseBoneSpace
    alphaBoolBlend {blendOption=Linear}
    baseBoneRef {boneName=pelvis}
    componentAppliedLinearAccClamp=<3000 3000 3000>
    componentLinearAccScale=<0.5 0.3 0.3>
    componentLinearVelScale=<0.5 0.3 0.3>
    overrideWorldGravity {z=-980}
    simSpaceSettings {
     maxAngularAcceleration=1000
     maxAngularVelocity=3
     maxLinearAcceleration=700
     maxLinearVelocity=500
     velocityScaleZ=0
     worldAlpha=1
     externalLinearDragV=<0.2 0.2>
    }
   }
   RigidBody(OverrideWorldGravity="0,0,-980") {
    bOverrideWorldGravity=true
    overridePhysicsAsset=&/M_MED_BadassCowboy_Body_Physics
    simulationSpace=BaseBoneSpace
    alphaBoolBlend {blendOption=Linear}
    baseBoneRef {boneName=root}
    componentAppliedLinearAccClamp=<250 250 250>
    componentLinearAccScale=<0.3 0.3 0.5>
    componentLinearVelScale=<0.3 0.3 0.5>
    overrideWorldGravity {z=-980}
   }
   RigidBody!(ComponentPose, Alpha, ExternalForce) -> (Pose) {
    bOverrideWorldGravity=true
    overridePhysicsAsset=&/M_MED_BadassCowboy_Lasso_Physics
    simulationSpace=BaseBoneSpace
    alphaBoolBlend {blendOption=Linear}
    baseBoneRef {boneName=root}
    componentAppliedLinearAccClamp=<250 250 250>
    componentLinearAccScale=<1 1 1>
    componentLinearVelScale=<1 1 1>
    overrideWorldGravity {z=-980}
   }
   n35 = Break Fort Anim Input Skydiving External Force(FortAnimInput_SkydivingExternalForce=Get SkydivingExternalForceInput()) -> (
    bUseSkydivingVectorForce=~
    SkydivingVectorForce=~
    HeadToPelvisDirection=~
    FloatingMultiplier=~
    FloatingAdditive=~
    DivingMultiplier=~
    DivingAdditive=~
    ParachutingMultiplier=~
    ParachutingAdditive=~
   )
   n36 = Get GravityOverride()
   RigidBody!(ComponentPose, Alpha, OverrideWorldGravity:Vector=@n36, ExternalForce=@n35.SkydivingVectorForce) -> (Pose) {
    bClampLinearTranslationLimitToRefPose=true
    bOverrideWorldGravity=true
    overridePhysicsAsset=&M_MED_BadassCowboy_Body_Perf_Physics
    simulationSpace=BaseBoneSpace
    alphaBoolBlend {blendOption=Linear}
    baseBoneRef {boneName=pelvis}
    componentAppliedLinearAccClamp=<3000 3000 3000>
    componentLinearAccScale=<0.5 0.3 0.3>
    componentLinearVelScale=<0.5 0.3 0.3>
    overrideWorldGravity {z=-980}
    simSpaceSettings {
     maxAngularAcceleration=1000
     maxAngularVelocity=3
     maxLinearAcceleration=700
     maxLinearVelocity=500
     velocityScaleZ=0
     worldAlpha=1
     externalLinearDragV=<0.2 0.2>
    }
   }
   n37 = "Transform (Modify) Bone - Bone: dyn_lasso_MSTR"(
    ComponentPose=Local To Component(LocalPose=Use cached pose 'PoseDriver_Cache'())
    Alpha=Get thigh_rot_up_r() -> (thigh_rot_up_r:double)
    Translation="0.891,0.645,-4.492"
    Rotation="-2.624,9.373,14.201"
    Scale="1,1,1"
   ) {
    rotationMode=BMM_Additive
    rotationSpace=BCS_BoneSpace
    translationMode=BMM_Additive
    translationSpace=BCS_BoneSpace
    alphaBoolBlend {blendOption=Linear}
    boneToModify {boneName=dyn_lasso_MSTR}
    rotation=<rot -2.624 9.3731 14.2011>
    translation=<0.891 0.645 -4.492>
   }
   n38 = "Transform (Modify) Bone - Bone: dyn_jacket_front_01_l"(ComponentPose=@n37, Alpha=Get thigh_rot_l(), Translation="0.000,-2,0.000", Rotation="-20,10.000,-0.000", Scale="1,1,1") {
    rotationMode=BMM_Additive
    rotationSpace=BCS_BoneSpace
    translationMode=BMM_Additive
    translationSpace=BCS_BoneSpace
    alphaBoolBlend {blendOption=Linear}
    boneToModify {boneName=dyn_jacket_front_01_l}
    rotation {pitch=-20, yaw=10}
    translation {y=-2}
   }
   n39 = RigidBody(ComponentPose=@n38, OverrideWorldGravity=@n34, ExternalForce=@n33.SkydivingVectorForce) {
    bOverrideWorldGravity=true
    overridePhysicsAsset=&/M_MED_BadassCowboy_Body_Perf_new_Physics
    simulationSpace=BaseBoneSpace
    alphaBoolBlend {blendOption=Linear}
    baseBoneRef {boneName=root}
    componentAppliedLinearAccClamp=<250 250 250>
    componentLinearAccScale=<1 1 1>
    componentLinearVelScale=<1 1 1>
    overrideWorldGravity {z=-980}
   }
   n40 = "Transform (Modify) Bone - Bone: dyn_lasso_MSTR"(
    ComponentPose=Local To Component(LocalPose=Use cached pose 'PoseDriver_Cache'())
    Alpha=Get thigh_rot_up_r()
    Translation="0.891,0.645,-4.492"
    Rotation="-2.624,9.373,14.201"
    Scale="1,1,1"
   ) {
    rotationMode=BMM_Additive
    rotationSpace=BCS_BoneSpace
    translationMode=BMM_Additive
    translationSpace=BCS_BoneSpace
    alphaBoolBlend {blendOption=Linear}
    boneToModify {boneName=dyn_lasso_MSTR}
    rotation=<rot -2.624 9.3731 14.2011>
    translation=<0.891 0.645 -4.492>
   }
   n41 = "Transform (Modify) Bone - Bone: dyn_jacket_front_01_l"(ComponentPose=@n40, Alpha=Get thigh_rot_l(), Translation="0.000,-2,0.000", Rotation="-20,10.000,-0.000", Scale="1,1,1") {
    rotationMode=BMM_Additive
    rotationSpace=BCS_BoneSpace
    translationMode=BMM_Additive
    translationSpace=BCS_BoneSpace
    alphaBoolBlend {blendOption=Linear}
    boneToModify {boneName=dyn_jacket_front_01_l}
    rotation {pitch=-20, yaw=10}
    translation {y=-2}
   }
   n42 = RigidBody(ComponentPose=@n41, OverrideWorldGravity=@n36, ExternalForce=@n35.SkydivingVectorForce) {
    bOverrideWorldGravity=true
    overridePhysicsAsset=&/M_MED_BadassCowboy_Body_Perf_Frontend_Physics
    simulationSpace=BaseBoneSpace
    alphaBoolBlend {blendOption=Linear}
    baseBoneRef {boneName=root}
    componentAppliedLinearAccClamp=<250 250 250>
    componentLinearAccScale=<1 1 1>
    componentLinearVelScale=<1 1 1>
    overrideWorldGravity {z=-980}
   }
   n43 = Blend Poses by bool(BlendPose_0=Component To Local(ComponentPose=@n42), BlendPose_1=Component To Local(ComponentPose=@n39), bActiveValue=Get bIsInFrontEnd()) {blendType=Linear}
   n44 = "Constraint - dyn_jacket_front_04_l (R)"(ComponentPose=Local To Component(LocalPose=Use cached pose 'PoseDriver_Cache'()), Alpha=1, ConstraintWeights_0=1) {
    alphaBoolBlend {blendOption=Linear}
    boneToModify {boneName=dyn_jacket_front_04_l}
    constraintSetup [{transformType=Rotation, targetBone {boneName=calf_l}}]
    constraintWeights [1]
   }
   n45 = "Constraint - dyn_jacket_front_04_r (R)"(ComponentPose=@n44, Alpha=1, ConstraintWeights_0=1) {
    alphaBoolBlend {blendOption=Linear}
    boneToModify {boneName=dyn_jacket_front_04_r}
    constraintSetup [{transformType=Rotation, targetBone {boneName=calf_r}}]
    constraintWeights [1]
   }
   n46 = "Constraint - dyn_jacket_rear_03_l (R)"(ComponentPose=@n45, Alpha=1, ConstraintWeights_0=1) {
    alphaBoolBlend {blendOption=Linear}
    boneToModify {boneName=dyn_jacket_rear_03_l}
    constraintSetup [{transformType=Rotation, targetBone {boneName=calf_l}}]
    constraintWeights [1]
   }
   n47 = "Constraint - dyn_jacket_rear_03_r (R)"(ComponentPose=@n46, Alpha=1, ConstraintWeights_0=1) {
    alphaBoolBlend {blendOption=Linear}
    boneToModify {boneName=dyn_jacket_rear_03_r}
    constraintSetup [{transformType=Rotation, targetBone {boneName=calf_r}}]
    constraintWeights [1]
   }
   n48 = "Constraint - dyn_jacket_side_03_l (R)"(ComponentPose=@n47, Alpha=1, ConstraintWeights_0=1) {
    alphaBoolBlend {blendOption=Linear}
    boneToModify {boneName=dyn_jacket_side_03_l}
    constraintSetup [{transformType=Rotation, targetBone {boneName=calf_l}}]
    constraintWeights [1]
   }
   n49 = "Constraint - dyn_jacket_side_03_r (R)"(ComponentPose=@n48, Alpha=1, ConstraintWeights_0=1) {
    alphaBoolBlend {blendOption=Linear}
    boneToModify {boneName=dyn_jacket_side_03_r}
    constraintSetup [{transformType=Rotation, targetBone {boneName=calf_r}}]
    constraintWeights [1]
   }
   n50 = Component To Local(ComponentPose=@n49)
   n51 = OR Boolean(A=NOT Boolean(A=Get bRigidBodyEnabled()), B="integer > integer"(A=Get CurrentLOD(), B=1) {is=title K2Node_CallFunction})
   n52 = "Constraint - dyn_holster_TGT (PPPPPPPPPPPP)"!(
    ComponentPose=Local To Component(LocalPose=Blend Poses by bool(BlendPose_0=@n50, BlendPose_1=@n43, bActiveValue=@n51) {blendType=Linear})
    Alpha=1
    ConstraintWeights_0=0.033
    ConstraintWeights_1:float=0.0017
    ConstraintWeights_2:float=0.2032
    ConstraintWeights_3:float=0.0031
    ConstraintWeights_4:float=0.0422
    ConstraintWeights_5:float=0.0623
    ConstraintWeights_6:float=0.3176
    ConstraintWeights_7:float=0.0015
    ConstraintWeights_8:float=0.0073
    ConstraintWeights_9:float=0.3116
    ConstraintWeights_10:float=0.0018
    ConstraintWeights_11:float=0.0146
   ) -> (Pose) {
    alphaBoolBlend {blendOption=Linear}
    boneToModify {boneName=dyn_holster_TGT}
    constraintSetup [
     {targetBone {boneName=dyn_jacket_rear_02_l}}
     {targetBone {boneName=dyn_jacket_rear_01_l}}
     {targetBone {boneName=dyn_jacket_front_03_l}}
     {targetBone {boneName=dyn_jacket_front_01_l}}
     {targetBone {boneName=dyn_jacket_rear_03_l}}
     {targetBone {boneName=dyn_jacket_front_02_l}}
     {targetBone {boneName=dyn_jacket_side_03_l}}
     {targetBone {boneName=dyn_jacket_rear_04_l}}
     {targetBone {boneName=dyn_jacket_front_04_l}}
     {targetBone {boneName=dyn_jacket_side_02_l}}
     {targetBone {boneName=pelvis}}
     {targetBone {boneName=dyn_jacket_side_01_l}}
    ]
    constraintWeights [1,1,1,1,1,1,1,1,1,1,1,1]
   }
   n53 = Look At - Bone: dyn_holster_AIM(ComponentPose=@n52, Alpha=1) {
    alphaBoolBlend {blendOption=Linear}
    boneToModify {boneName=dyn_holster_AIM}
    lookAtLocation {x=0}
    lookAtTarget {boneReference {boneName=dyn_holster_TGT}}
    lookAt_Axis {axis=<1 0>}
   }
   n54 = "Constraint - belt_back_TGT (PPPPPPP)"!(
    ComponentPose=@n53
    Alpha=1
    ConstraintWeights_0=0.0641
    ConstraintWeights_1:float=0.5145
    ConstraintWeights_2:float=0.1436
    ConstraintWeights_3:float=0.0834
    ConstraintWeights_4:float=0.0179
    ConstraintWeights_5:float=0.0305
    ConstraintWeights_6:float=0.146
   ) -> (Pose) {
    alphaBoolBlend {blendOption=Linear}
    boneToModify {boneName=belt_back_TGT}
    constraintSetup [
     {targetBone {boneName=dyn_jacket_rear_02_l}}
     {targetBone {boneName=dyn_jacket_rear_01_l}}
     {targetBone {boneName=dyn_jacket_side_01_l}}
     {targetBone {boneName=dyn_jacket_rear_01_r}}
     {targetBone {boneName=thigh_twist_01_l}}
     {targetBone {boneName=dyn_jacket_side_02_l}}
     {targetBone {boneName=pelvis}}
    ]
    constraintWeights [1,1,1,1,1,1,1]
   }
   n55 = Look At - Bone: belt_back(ComponentPose=@n54, Alpha=1) {
    alphaBoolBlend {blendOption=Linear}
    boneToModify {boneName=belt_back}
    lookAtLocation {x=0}
    lookAtTarget {boneReference {boneName=belt_back_TGT}}
    lookAt_Axis {axis=<1 0>}
   }
   n56 = "Transform (Modify) Bone - Bone: belt_front_MSTR"(
    ComponentPose=@n55
    Alpha=Get thigh_rot_up_l_a() -> (thigh_rot_up_l_a:double)
    Translation="-5.042,0.229,-0.085"
    Rotation="20.000,-5.174,-16.338"
    Scale="1,1,1"
   ) {
    rotationMode=BMM_Additive
    rotationSpace=BCS_BoneSpace
    translationMode=BMM_Additive
    translationSpace=BCS_BoneSpace
    alphaBoolBlend {blendOption=Linear}
    boneToModify {boneName=belt_front_MSTR}
    rotation=<rot 20 -5.174 -16.3382>
    translation=<-5.0422 0.229 -0.0851>
   }
   n57 = Get thigh_rot_up_l() -> (thigh_rot_up_l:double)
   n58 = "Transform (Modify) Bone - Bone: dyn_holster_MSTR"(ComponentPose=@n56, Alpha=@n57, Translation="-4.8,-2.027,1.838", Rotation="0,0,0", Scale="1,1,1") {
    rotationMode=BMM_Additive
    rotationSpace=BCS_BoneSpace
    translationMode=BMM_Additive
    translationSpace=BCS_BoneSpace
    alphaBoolBlend {blendOption=Linear}
    boneToModify {boneName=dyn_holster_MSTR}
    translation=<-4.8 -2.027 1.838>
   }
   n59 = "Transform (Modify) Bone - Bone: belt_back_MSTR"(ComponentPose=@n58, Alpha=@n57, Translation="-2.294,-0.000,0.058", Rotation="5.000,0.000,0.000", Scale="1,1,1") {
    rotationMode=BMM_Additive
    rotationSpace=BCS_BoneSpace
    translationMode=BMM_Additive
    translationSpace=BCS_BoneSpace
    alphaBoolBlend {blendOption=Linear}
    boneToModify {boneName=belt_back_MSTR}
    rotation {pitch=5}
    translation {x=-2.294, z=0.058}
   }
   n60 = "Constraint - dyn_lasso_TGT (PPPPPPPPPP)"!(
    ComponentPose=@n59
    Alpha=1
    ConstraintWeights_0=0.0057
    ConstraintWeights_1:float=0.0567
    ConstraintWeights_2:float=0.3554
    ConstraintWeights_3:float=0.1988
    ConstraintWeights_4:float=0.0429
    ConstraintWeights_5:float=0.1232
    ConstraintWeights_6:float=0.1603
    ConstraintWeights_7:float=0.0211
    ConstraintWeights_8:float=0.03
    ConstraintWeights_9:float=0.006
   ) -> (Pose) {
    alphaBoolBlend {blendOption=Linear}
    boneToModify {boneName=dyn_lasso_TGT}
    constraintSetup [
     {targetBone {boneName=pelvis}}
     {targetBone {boneName=dyn_jacket_side_01_r}}
     {targetBone {boneName=dyn_jacket_side_02_r}}
     {targetBone {boneName=dyn_jacket_side_03_r}}
     {targetBone {boneName=dyn_jacket_front_01_r}}
     {targetBone {boneName=dyn_jacket_front_02_r}}
     {targetBone {boneName=dyn_jacket_front_03_r}}
     {targetBone {boneName=dyn_jacket_rear_01_r}}
     {targetBone {boneName=dyn_jacket_rear_03_r}}
     {targetBone {boneName=dyn_jacket_rear_02_r}}
    ]
    constraintWeights [1,1,1,1,1,1,1,1,1,1]
   }
   n61 = "Constraint - dyn_rope_TGT (PPPPPPPPPPPPP)"!(
    ComponentPose=@n60
    Alpha=1
    ConstraintWeights_0=0.0031
    ConstraintWeights_1:float=0
    ConstraintWeights_2:float=0
    ConstraintWeights_3:float=0.0344
    ConstraintWeights_4:float=0.2928
    ConstraintWeights_5:float=0.1845
    ConstraintWeights_6:float=0.0002
    ConstraintWeights_7:float=0.0004
    ConstraintWeights_8:float=0
    ConstraintWeights_9:float=0.1308
    ConstraintWeights_10:float=0.2094
    ConstraintWeights_11:float=0
    ConstraintWeights_12:float=0.1443
   ) -> (Pose) {
    alphaBoolBlend {blendOption=Linear}
    boneToModify {boneName=dyn_rope_TGT}
    constraintSetup [
     {targetBone {boneName=pelvis}}
     {targetBone {boneName=dyn_jacket_rear_01_l}}
     {targetBone {boneName=dyn_jacket_rear_03_l}}
     {targetBone {boneName=dyn_jacket_side_01_r}}
     {targetBone {boneName=dyn_jacket_side_02_r}}
     {targetBone {boneName=dyn_jacket_side_03_r}}
     {targetBone {boneName=dyn_jacket_front_02_r}}
     {targetBone {boneName=dyn_jacket_front_03_r}}
     {targetBone {boneName=dyn_jacket_front_04_r}}
     {targetBone {boneName=dyn_jacket_rear_01_r}}
     {targetBone {boneName=dyn_jacket_rear_03_r}}
     {targetBone {boneName=dyn_jacket_rear_02_l}}
     {targetBone {boneName=dyn_jacket_rear_02_r}}
    ]
    constraintWeights [1,1,1,1,1,1,1,1,1,1,1,1,1]
   }
   n62 = "Transform (Modify) Bone - Bone: dyn_lasso_MSTR"(ComponentPose=@n61, Alpha=Get thigh_rot_r(), Translation="0,0,0", Rotation="0,0,15", Scale="1,1,1") {
    rotationMode=BMM_Additive
    rotationSpace=BCS_BoneSpace
    alphaBoolBlend {blendOption=Linear}
    boneToModify {boneName=dyn_lasso_MSTR}
    rotation {roll=15}
   }
   n63 = "Transform (Modify) Bone - Bone: dyn_lasso_TGT"!(ComponentPose=@n62, bAlphaBoolEnabled:bool=True, Translation="0,2,0", Rotation="0,0,0", Scale="1,1,1") -> (Pose) {
    alphaInputType=Bool
    translationMode=BMM_Additive
    translationSpace=BCS_BoneSpace
    alphaBoolBlend {blendOption=Linear}
    boneToModify {boneName=dyn_lasso_TGT}
    translation {y=2}
   }
   n64 = "Transform (Modify) Bone - Bone: dyn_lasso_02"!(ComponentPose=@n63, bAlphaBoolEnabled:bool=True, Translation="0,0,0", Rotation="0,10,0", Scale="1,1,1") -> (Pose) {
    alphaInputType=Bool
    rotationMode=BMM_Additive
    rotationSpace=BCS_BoneSpace
    translationMode=BMM_Additive
    translationSpace=BCS_BoneSpace
    alphaBoolBlend {blendInTime=0.1, blendOption=Linear, blendOutTime=0.1}
    boneToModify {boneName=dyn_lasso_02}
    rotation {yaw=10}
   }
   n65 = "Transform (Modify) Bone - Bone: dyn_lasso_02"(ComponentPose=@n64, bAlphaBoolEnabled=True, Translation="0,0,0", Rotation="-4.943,15.093,-0.752", Scale="1,1,1") {
    alphaInputType=Bool
    rotationMode=BMM_Additive
    rotationSpace=BCS_BoneSpace
    translationMode=BMM_Additive
    translationSpace=BCS_BoneSpace
    alphaBoolBlend {blendInTime=0.1, blendOption=Linear, blendOutTime=0.1}
    boneToModify {boneName=dyn_lasso_02}
    rotation=<rot -4.943 15.093 -0.752>
   }
   n66 = "Transform (Modify) Bone - Bone: dyn_rope_TGT"!(
    ComponentPose=@n65
    bAlphaBoolEnabled:bool=Get tgt MOVER() -> (tgt MOVER:bool)
    Translation="-0.000,-0.5,-4"
    Rotation="0,0,0"
    Scale="1,1,1"
   ) -> (Pose) {
    alphaInputType=Bool
    translationMode=BMM_Additive
    translationSpace=BCS_BoneSpace
    alphaBoolBlend {blendInTime=0.1, blendOption=Linear, blendOutTime=0.1}
    boneToModify {boneName=dyn_rope_TGT}
    translation {y=-0.5, z=-4}
   }
   n67 = "Transform (Modify) Bone - Bone: dyn_rope_TGT"(ComponentPose=@n66, bAlphaBoolEnabled=Get bIsCrouching(), Translation="0,0,-3", Rotation="0,0,0", Scale="1,1,1") {
    alphaInputType=Bool
    translationMode=BMM_Additive
    translationSpace=BCS_BoneSpace
    alphaBoolBlend {blendInTime=0.1, blendOption=Linear, blendOutTime=0.1}
    boneToModify {boneName=dyn_rope_TGT}
    translation {z=-3}
   }
   n68 = Look At - Bone: dyn_lasso_MSTR(ComponentPose=@n67, Alpha=1) {
    alphaBoolBlend {blendOption=Linear}
    boneToModify {boneName=dyn_lasso_MSTR}
    lookAtLocation=<0 2 -3>
    lookAtTarget {boneReference {boneName=dyn_lasso_TGT}}
    lookAt_Axis {axis=<1 0>}
   }
   n69 = Look At - Bone: dyn_rope_01(ComponentPose=@n68, Alpha=1) {
    alphaBoolBlend {blendOption=Linear}
    boneToModify {boneName=dyn_rope_01}
    lookAtLocation=<0 -0.142 10>
    lookAtTarget {boneReference {boneName=dyn_rope_TGT}}
    lookAt_Axis {axis=<1 0>}
   }
   n70 = "Transform (Modify) Bone - Bone: dyn_rope_01_B"!(ComponentPose=@n69, bAlphaBoolEnabled:bool=True, Translation="0,0,0", Rotation="0,5,0", Scale="1,1,1") -> (Pose) {
    alphaInputType=Bool
    rotationMode=BMM_Additive
    rotationSpace=BCS_ParentBoneSpace
    translationMode=BMM_Additive
    translationSpace=BCS_BoneSpace
    alphaBoolBlend {blendInTime=0.1, blendOption=Linear, blendOutTime=0.1}
    boneToModify {boneName=dyn_rope_01_B}
    rotation {yaw=5}
   }
   n71 = "Transform (Modify) Bone - Bone: dyn_rope_02"!(ComponentPose=@n70, bAlphaBoolEnabled:bool=True, Translation="0,0,0", Rotation="0,15,0", Scale="1,1,1") -> (Pose) {
    alphaInputType=Bool
    rotationMode=BMM_Additive
    rotationSpace=BCS_BoneSpace
    translationMode=BMM_Additive
    translationSpace=BCS_BoneSpace
    alphaBoolBlend {blendInTime=0.1, blendOption=Linear, blendOutTime=0.1}
    boneToModify {boneName=dyn_rope_02}
    rotation {yaw=15}
   }
   n72 = Anim Dynamics - Bone: dyn_lasso_01(ComponentPose=@n71, Alpha=1, ExternalForce="0,0,0") {
    angularDampingOverride=0.9999
    angularSpringConstant=250
    bAngularSpring=true
    bGravityOverrideInSimSpace=true
    bOverrideAngularDamping=true
    bUseGravityOverride=true
    simulationSpace=RootRelative
    alphaBoolBlend {blendOption=Linear}
    boundBone {boneName=dyn_lasso_01}
    componentAppliedLinearAccClamp=<3500 3500 3500>
    componentLinearAccScale=<0.4 0.4 0.4>
    componentLinearVelScale=<0.4 0.4 0.4>
    gravityOverride {z=-980}
    physicsBodyDefinitions [{
      boundBone {boneName=dyn_lasso_01}
      boxExtents=<15 15 15>
      constraintSetup {angularLimitsMax=<5 10 15>, angularLimitsMin=<-5 -10 0>, angularTarget=<0 0 0>}
      localJointOffset=<2 0 0>
     }]
    relativeSpaceBone {boneName=root}
   }
   n73 = Anim Dynamics - Bone: dyn_lasso_02(ComponentPose=@n72, Alpha=1, ExternalForce="0,0,0") {
    angularDampingOverride=0.9999
    angularSpringConstant=250
    bAngularSpring=true
    bGravityOverrideInSimSpace=true
    bOverrideAngularDamping=true
    bUseGravityOverride=true
    simulationSpace=RootRelative
    alphaBoolBlend {blendOption=Linear}
    boundBone {boneName=dyn_lasso_02}
    componentAppliedLinearAccClamp=<3500 3500 3500>
    componentLinearAccScale=<0.6 0.6 0.6>
    componentLinearVelScale=<0.6 0.6 0.6>
    gravityOverride {z=-980}
    physicsBodyDefinitions [{
      boundBone {boneName=dyn_lasso_02}
      boxExtents=<10 10 15>
      constraintSetup {angularLimitsMax=<5 15 35>, angularLimitsMin=<-5 -15 0>, angularTarget=<0 0 0>}
      localJointOffset=<2 0 0>
     }]
    relativeSpaceBone {boneName=root}
   }
   n74 = Anim Dynamics - Bone: dyn_lasso_03(ComponentPose=@n73, Alpha=1, ExternalForce="0,0,0") {
    angularDampingOverride=0.9999
    angularSpringConstant=250
    bAngularSpring=true
    bGravityOverrideInSimSpace=true
    bOverrideAngularDamping=true
    bUseGravityOverride=true
    simulationSpace=RootRelative
    alphaBoolBlend {blendOption=Linear}
    boundBone {boneName=dyn_lasso_03}
    componentAppliedLinearAccClamp=<3500 3500 3500>
    componentLinearAccScale=<0.6 0.6 0.6>
    componentLinearVelScale=<0.6 0.6 0.6>
    gravityOverride {z=-980}
    physicsBodyDefinitions [{
      boundBone {boneName=dyn_lasso_03}
      boxExtents=<7 7 10>
      constraintSetup {angularLimitsMax=<5 15 45>, angularLimitsMin=<-5 -15 -5>, angularTarget=<0 0 0>}
      localJointOffset=<4 0 0>
     }]
    relativeSpaceBone {boneName=root}
   }
   n75 = Anim Dynamics - Bone: dyn_rope_01_B(ComponentPose=@n74, Alpha=1, ExternalForce="0,0,0") {
    angularDampingOverride=0.9999
    angularSpringConstant=250
    bAngularSpring=true
    bGravityOverrideInSimSpace=true
    bOverrideAngularDamping=true
    bUseGravityOverride=true
    simulationSpace=RootRelative
    alphaBoolBlend {blendOption=Linear}
    boundBone {boneName=dyn_rope_01_B}
    componentAppliedLinearAccClamp=<3500 3500 3500>
    componentLinearAccScale=<0.6 0.6 0.6>
    componentLinearVelScale=<0.6 0.6 0.6>
    gravityOverride {z=-980}
    physicsBodyDefinitions [{
      boundBone {boneName=dyn_rope_01_B}
      boxExtents=<10 15 15>
      constraintSetup {angularLimitsMax=<5 10 25>, angularLimitsMin=<-5 -15 0>, angularTarget=<0 0 0>}
      localJointOffset=<2 0 0>
     }]
    relativeSpaceBone {boneName=root}
   }
   n76 = Anim Dynamics - Bone: dyn_rope_02(ComponentPose=@n75, Alpha=1, ExternalForce="0,0,0") {
    angularDampingOverride=0.9999
    angularSpringConstant=250
    bAngularSpring=true
    bGravityOverrideInSimSpace=true
    bOverrideAngularDamping=true
    bUseGravityOverride=true
    simulationSpace=RootRelative
    alphaBoolBlend {blendOption=Linear}
    boundBone {boneName=dyn_rope_02}
    componentAppliedLinearAccClamp=<3500 3500 3500>
    componentLinearAccScale=<0.8 0.8 0.8>
    componentLinearVelScale=<0.8 0.8 0.8>
    gravityOverride {z=-980}
    physicsBodyDefinitions [{
      boundBone {boneName=dyn_rope_02}
      boxExtents=<10 15 15>
      constraintSetup {angularLimitsMax=<5 10 35>, angularLimitsMin=<-5 -15 0>, angularTarget=<0 0 0>}
      localJointOffset=<3 0 0>
     }]
    relativeSpaceBone {boneName=root}
   }
   n77 = Anim Dynamics - Bone: dyn_rope_03(ComponentPose=@n76, Alpha=1, ExternalForce="0,0,0") {
    angularDampingOverride=0.9999
    angularSpringConstant=250
    bAngularSpring=true
    bGravityOverrideInSimSpace=true
    bOverrideAngularDamping=true
    bUseGravityOverride=true
    simulationSpace=RootRelative
    alphaBoolBlend {blendOption=Linear}
    boundBone {boneName=dyn_rope_03}
    componentAppliedLinearAccClamp=<3500 3500 3500>
    componentLinearAccScale=<0.8 0.8 0.8>
    componentLinearVelScale=<0.8 0.8 0.8>
    gravityOverride {z=-980}
    physicsBodyDefinitions [{
      boundBone {boneName=dyn_rope_03}
      boxExtents=<10 15 15>
      constraintSetup {angularLimitsMax=<5 10 35>, angularLimitsMin=<-5 -15 0>, angularTarget=<0 0 0>}
      localJointOffset=<4 0 0>
     }]
    relativeSpaceBone {boneName=root}
   }
   n78 = "Constraint - dyn_collar_bk_MATR (PP)"!(ComponentPose=@n77, Alpha=1, ConstraintWeights_0=0.57, ConstraintWeights_1:float=0.43) -> (Pose) {
    alphaBoolBlend {blendOption=Linear}
    boneToModify {boneName=dyn_collar_bk_MATR}
    constraintSetup [{targetBone {boneName=spine_05}}, {targetBone {boneName=neck_01}}]
    constraintWeights [1,1]
   }
   n79 = "Constraint - dyn_collar_r_01_MATR (PP)"!(ComponentPose=@n78, Alpha=1, ConstraintWeights_0=0.87, ConstraintWeights_1:float=0.13) -> (Pose) {
    alphaBoolBlend {blendOption=Linear}
    boneToModify {boneName=dyn_collar_r_01_MATR}
    constraintSetup [{targetBone {boneName=spine_05}}, {targetBone {boneName=clavicle_r}}]
    constraintWeights [1,1]
   }
   n80 = "Constraint - dyn_collar_r_02_MATR (PP)"!(ComponentPose=@n79, Alpha=1, ConstraintWeights_0=0.23, ConstraintWeights_1:float=0.77) -> (Pose) {
    alphaBoolBlend {blendOption=Linear}
    boneToModify {boneName=dyn_collar_r_02_MATR}
    constraintSetup [{targetBone {boneName=clavicle_r}}, {targetBone {boneName=spine_05}}]
    constraintWeights [1,1]
   }
   n81 = "Constraint - dyn_collar_r_03_MATR (PP)"!(ComponentPose=@n80, Alpha=1, ConstraintWeights_0=0.18, ConstraintWeights_1:float=0.82) -> (Pose) {
    alphaBoolBlend {blendOption=Linear}
    boneToModify {boneName=dyn_collar_r_03_MATR}
    constraintSetup [{targetBone {boneName=clavicle_r}}, {targetBone {boneName=spine_05}}]
    constraintWeights [1,1]
   }
   n82 = "Constraint - dyn_collar_l_01_MATR (PP)"!(ComponentPose=@n81, Alpha=1, ConstraintWeights_0=0.87, ConstraintWeights_1:float=0.13) -> (Pose) {
    alphaBoolBlend {blendOption=Linear}
    boneToModify {boneName=dyn_collar_l_01_MATR}
    constraintSetup [{targetBone {boneName=spine_05}}, {targetBone {boneName=clavicle_l}}]
    constraintWeights [1,1]
   }
   n83 = "Constraint - dyn_collar_l_02_MATR (PP)"!(ComponentPose=@n82, Alpha=1, ConstraintWeights_0=0.23, ConstraintWeights_1:float=0.77) -> (Pose) {
    alphaBoolBlend {blendOption=Linear}
    boneToModify {boneName=dyn_collar_l_02_MATR}
    constraintSetup [{targetBone {boneName=clavicle_l}}, {targetBone {boneName=spine_05}}]
    constraintWeights [1,1]
   }
   n84 = "Constraint - dyn_collar_l_03_MATR (PP)"!(ComponentPose=@n83, Alpha=1, ConstraintWeights_0=0.18, ConstraintWeights_1:float=0.82) -> (Pose) {
    alphaBoolBlend {blendOption=Linear}
    boneToModify {boneName=dyn_collar_l_03_MATR}
    constraintSetup [{targetBone {boneName=clavicle_l}}, {targetBone {boneName=spine_05}}]
    constraintWeights [1,1]
   }
   n85 = Get arm r rot() -> (arm r rot:double)
   n86 = "Transform (Modify) Bone - Bone: dyn_collar_r_02_MATR"(ComponentPose=@n84, Alpha=@n85, Translation="0,0,0", Rotation="-10,0,0", Scale="1,1,1") {
    rotationMode=BMM_Additive
    rotationSpace=BCS_BoneSpace
    translationMode=BMM_Additive
    translationSpace=BCS_BoneSpace
    alphaBoolBlend {blendOption=Linear}
    boneToModify {boneName=dyn_collar_r_02_MATR}
    rotation {pitch=-10}
   }
   n87 = "Transform (Modify) Bone - Bone: dyn_collar_r_03_MATR"(ComponentPose=@n86, Alpha=@n85, Translation="0,0,0", Rotation="-35,0,0.109", Scale="1,1,1") {
    rotationMode=BMM_Additive
    rotationSpace=BCS_BoneSpace
    translationMode=BMM_Additive
    translationSpace=BCS_BoneSpace
    alphaBoolBlend {blendOption=Linear}
    boneToModify {boneName=dyn_collar_r_03_MATR}
    rotation {pitch=-35, roll=0.109}
   }
   n88 = RigidBody!(ComponentPose=@n87, Alpha, ExternalForce) -> (Pose) {
    bClampLinearTranslationLimitToRefPose=true
    bOverrideWorldGravity=true
    overridePhysicsAsset=&/M_MED_BadassCowboy_dyn_collar
    simulationSpace=BaseBoneSpace
    alphaBoolBlend {blendOption=Linear}
    baseBoneRef {boneName=spine_05}
    componentAppliedLinearAccClamp=<500 500 500>
    componentLinearAccScale=<0.1 0.1 0.1>
    componentLinearVelScale=<0.1 0.1 0.1>
    overrideWorldGravity {z=-980}
   }
   n89 = Get arm r rot()
   n90 = "Transform (Modify) Bone - Bone: dyn_collar_bk"(ComponentPose=@n88, Alpha=@n89, Translation="0,0,0", Rotation="0,0,0", Scale="0.750,1.000,1.000") {
    rotationMode=BMM_Additive
    rotationSpace=BCS_BoneSpace
    scaleMode=BMM_Additive
    scaleSpace=BCS_BoneSpace
    translationMode=BMM_Additive
    translationSpace=BCS_BoneSpace
    alphaBoolBlend {blendOption=Linear}
    boneToModify {boneName=dyn_collar_bk}
    scale {x=0.75}
   }
   n91 = "Transform (Modify) Bone - Bone: dyn_collar_r_01"(ComponentPose=@n90, Alpha=@n89, Translation="0,0,0", Rotation="0,0,0", Scale="0.750,1.000,1.000") {
    rotationMode=BMM_Additive
    rotationSpace=BCS_BoneSpace
    scaleMode=BMM_Additive
    scaleSpace=BCS_BoneSpace
    translationMode=BMM_Additive
    translationSpace=BCS_BoneSpace
    alphaBoolBlend {blendOption=Linear}
    boneToModify {boneName=dyn_collar_r_01}
    scale {x=0.75}
   }
   n92 = Anim Dynamics - Bone: dyn_pouch_r(ComponentPose=@n91, Alpha=1, ExternalForce="0,0,0") {
    angularDampingOverride=0.999
    angularSpringConstant=50
    bAngularSpring=true
    bGravityOverrideInSimSpace=true
    bOverrideAngularDamping=true
    bUseGravityOverride=true
    gravityScale=0.35
    simulationSpace=RootRelative
    alphaBoolBlend {blendOption=Linear}
    boundBone {boneName=dyn_pouch_r}
    componentAppliedLinearAccClamp=<3000 3000 3000>
    componentLinearAccScale=<0.2 0.2 0.2>
    componentLinearVelScale=<0.2 0.2 0.2>
    gravityOverride {z=-980}
    physicsBodyDefinitions [{
      sphereCollisionRadius=0
      boundBone {boneName=dyn_pouch_r}
      boxExtents=<10 30 30>
      constraintSetup {angularLimitsMax=<0 20 15>, angularLimitsMin=<0 -20 0>, angularTarget=<0 0 0>}
      localJointOffset=<2 0 0>
     }]
   }
   n93 = Anim Dynamics - Bone: dyn_shoulder_sleeve_02_l(ComponentPose=@n92, Alpha=0.5, ExternalForce="0,0,0") {
    angularDampingOverride=0.999
    angularSpringConstant=250
    bAngularSpring=true
    bGravityOverrideInSimSpace=true
    bOverrideAngularDamping=true
    bOverrideLinearDamping=true
    bUseGravityOverride=true
    gravityScale=0.35
    lODThreshold=1
    linearDampingOverride=0.5
    windScale=0.5
    alphaBoolBlend {blendOption=Linear}
    boundBone {boneName=dyn_shoulder_sleeve_02_l}
    componentAppliedLinearAccClamp=<1000 1000 1000>
    componentLinearAccScale=<1 1 1>
    componentLinearVelScale=<1 1 1>
    gravityOverride {z=-980}
    physicsBodyDefinitions [{
      sphereCollisionRadius=0
      boundBone {boneName=dyn_shoulder_sleeve_02_l}
      boxExtents=<10 20 20>
      constraintSetup {angularLimitsMax=<0 8 5>, angularLimitsMin=<0 -60 -5>, angularTarget=<0 0 0>}
      localJointOffset=<6 0 -2>
     }]
   }
   n94 = Anim Dynamics - Bone: dyn_shoulder_sleeve_02_r(ComponentPose=@n93, Alpha=0.5, ExternalForce="0,0,0") {
    angularDampingOverride=0.999
    angularSpringConstant=250
    bAngularSpring=true
    bGravityOverrideInSimSpace=true
    bOverrideAngularDamping=true
    bOverrideLinearDamping=true
    bUseGravityOverride=true
    gravityScale=0.35
    lODThreshold=1
    linearDampingOverride=0.5
    windScale=0.5
    alphaBoolBlend {blendOption=Linear}
    boundBone {boneName=dyn_shoulder_sleeve_02_r}
    componentAppliedLinearAccClamp=<1000 1000 1000>
    componentLinearAccScale=<1 1 1>
    componentLinearVelScale=<1 1 1>
    gravityOverride {z=-980}
    physicsBodyDefinitions [{
      sphereCollisionRadius=0
      boundBone {boneName=dyn_shoulder_sleeve_02_r}
      boxExtents=<10 20 20>
      constraintSetup {angularLimitsMax=<0 8 5>, angularLimitsMin=<0 -60 -5>, angularTarget=<0 0 0>}
      localJointOffset=<-6 0 2>
     }]
   }
   n95 = Get PectStretchAlpha() -> (PectStretchAlpha:double)
   n96 = "Transform (Modify) Bone - Bone: deltoid_r_MSTR"(ComponentPose=@n94, Alpha=@n95, Translation="40,-10,-15", Rotation="0,0,0", Scale="1,1,1") {
    rotationMode=BMM_Additive
    rotationSpace=BCS_BoneSpace
    translationMode=BMM_Additive
    translationSpace=BCS_BoneSpace
    alphaBoolBlend {blendOption=Linear}
    boneToModify {boneName=deltoid_r_MSTR}
    translation=<40 -10 -15>
   }
   n97 = "Transform (Modify) Bone - Bone: tricep_01_r_MSTR"(ComponentPose=@n96, Alpha=@n95, Translation="0,0,20", Rotation="0,0,0", Scale="1,1,1") {
    rotationMode=BMM_Additive
    rotationSpace=BCS_BoneSpace
    translationMode=BMM_Additive
    translationSpace=BCS_BoneSpace
    alphaBoolBlend {blendOption=Linear}
    boneToModify {boneName=tricep_01_r_MSTR}
    translation {z=20}
   }
   Output Pose(Result=Component To Local(ComponentPose=@n97)) {layerGroup=None}
   Save cached pose 'PoseDriver_Cache'(Pose=Copy Pose From Mesh() {bUseAttachedParent=true}) {cacheName=PoseDriver_Cache}
   Save cached pose 'CopyPose_Cache'(Pose=Copy Pose From Mesh() {bUseAttachedParent=true}) {cacheName=CopyPose_Cache}
  ]}
 EventGraph {nodes [
   Event BlueprintUpdateAnimation()
   > n1 = Sequence!(execute) -> (then_0, then_1, then_2:Exec, then_3:Exec)
   n2 = BreakTransform(InTransform=GetDeltaTransformFromRefPose(self=GetOwnerSkeletalMeshComponent(), BoneName=thigh_r, BaseName=pelvis))
   > Set thigh_rot_up_r(thigh_rot_up_r:double=MapRangeClamped(Value=BreakRotator(InRot=@n2.Rotation).Yaw, InRangeA=0, InRangeB=-100)) -> (Output_Get:double)
   n3 = GetOwnerSkeletalMeshComponent()
   n4 = BreakRotator(InRot=BreakTransform(InTransform=GetDeltaTransformFromRefPose(self=@n3, BoneName=thigh_l, BaseName=pelvis)).Rotation)
   > Set thigh_rot_up_l_a(thigh_rot_up_l_a:double=MapRangeClamped(Value=@n4.Yaw, InRangeA=0, InRangeB=-30)) -> (Output_Get:double)
   > Set thigh_rot_up_l(thigh_rot_up_l:double=MapRangeClamped(Value=@n4.Yaw, InRangeA=0, InRangeB=-110)) -> (Output_Get:double)
   > n5 = Gravity Override State Function DEPRECATED(CharacterPartAnimInstance=Self-Reference(), GameTimeInSeconds=GetGameTimeInSeconds(), GravityOverrideParams=Get Gravity Override Params())
   > Set GravityOverride(GravityOverride=@n5)
   n6 = GetPartSkeletalMeshComponent()
   n7 = TransformFromBoneSpace!(
    execute=@n1.then_1
    self:Skinned Mesh Component Object Reference=@n6
    BoneName:Name=dyn_tricep_measure_r
    InPosition:Vector="0, 0, 0"
    InRotation:Rotator="0, 0, 0"
   ) -> (then, OutPosition:Vector, OutRotation:Rotator)
   > n8 = TransformFromBoneSpace(self=@n6, BoneName=dyn_pec_measure_r, InPosition="0, 0, 0", InRotation="0, 0, 0")
   > Set PectStretchAlpha(PectStretchAlpha:double=MapRangeClamped(Value="Distance (Vector)"!(V1:Vector=@n7, V2:Vector=@n8) -> (ReturnValue:double), InRangeA=20, InRangeB=5)) -> (Output_Get:double)
   n9 = Get Skel_Mesh()
   Set thigh_rot_l(
    execute=@n1.then_2
    thigh_rot_l=GetAlphaFromReferenceTransformDelta(SkeletalMeshComponent=@n9, BoneName=thigh_l, BaseName=pelvis, InputRange=(X=-20,Y=-115), TwistAxis="1,0,0")
   )
   > Set thigh_rot_r(thigh_rot_r=GetAlphaFromReferenceTransformDelta(SkeletalMeshComponent=@n9, BoneName=thigh_r, BaseName=pelvis, InputRange=(X=-20,Y=-100), TwistAxis="1,0,0"))
   > Set arm r rot(arm r rot:double=GetAlphaFromReferenceTransformDelta(SkeletalMeshComponent=@n9, BoneName=upperarm_r, BaseName=spine_05, InputRange=(X=70,Y=120), TwistAxis="1,0,0")) -> (Output_Get:double)
   n10 = Branch(execute=@n1.then_3, Condition=Get bIsCrouching())
   > Set tgt MOVER(tgt MOVER:bool=false) -> (Output_Get:bool)
   n11 = Branch(execute=@n10.else, Condition=Get bIsMoving2D())
   > Set tgt MOVER(tgt MOVER=true)
   Set tgt MOVER(execute=@n11.else, tgt MOVER=false)
   Event OnParachuteOpened()
   > n12 = Delay(Duration=1.5)
   ResetDynamics(execute=[@n12.then, @n13.then, @n15.then, @n16.then], InTeleportType=TeleportPhysics)
   > ResetDynamics(InTeleportType=ResetPhysics)
   n13 = Event OnBeginSkydiving()
   Event BlueprintInitializeAnimation()
   > Set Skel_Mesh(Skel_Mesh=GetOwnerSkeletalMeshComponent())
   Event OnBeginTargeting() {is=title K2Node_Event}
   n14 = Get bIsCrouching()
   > Branch(Condition=AND Boolean(A=Get bIsTargeting(), B=NOT Boolean(A=@n14)))
   n15 = Event OnEndSkydiving()
   n16 = Event OnBeginCrouching()
   n17 = Branch(Condition=Get bIsSwimSprinting() -> (bIsSwimSprinting:bool))
   n18 = Branch(execute=@n17.else, Condition=Get bIsSkyDiving())
   > n19 = Branch(Condition=Get bIsParachuteOpened())
   Set GravityOverride(execute=[@n18.else, @n19.then], GravityOverride="0, 0,-980")
   n20 = Branch(execute=@n19.else, Condition="float > float"(A=Get PawnVelocityZ(), B=50) {is=title K2Node_CallFunction})
   > Set GravityOverride(GravityOverride="0, 0, 0")
   Set GravityOverride(execute=@n20.else, GravityOverride="0, 0,-300")
   Event OnEndCrouching()
  ]}
]
variables [
 arm r rot {type=number}
 crouching {type=number}
 Gravity Override Params {type=GravityOverrideParamsStruct}
 GravityOverride {type=Vector}
 PectStretchAlpha {type=number}
 Skel_Mesh {type=SkeletalMeshComponent}
 tgt MOVER {type=boolean}
 thigh_rot_l {type=number}
 thigh_rot_r {type=number}
 thigh_rot_up_l {type=number}
 thigh_rot_up_l_a {type=number}
 thigh_rot_up_r {type=number}
]
