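# Physics asset for the medium male BadassCowboy player body, providing collision volumes that, by its name, the character's cape collides against.
# Capsules cover the pelvis, spine, clavicles and neck, with a sphere on the head.
# Extra capsules on the dynamic holster, pouch, rope and rear jacket bones account for costume bulk.
# Arm and leg bodies hold no shapes, and there are no constraints, so it is a collider set rather than a ragdoll.
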PhysicsAsset /Game/Characters/Player/Male/Medium/Bodies/M_MED_BadassCowboy/Meshes/M_MED_BadassCowboy_CapeColliders
bodies {
 calf_l []
 calf_r []
 clavicle_l [{shapeType=Capsule, length=7.7636, radius=11.1427, center=<9.9642 -0.7741 -1.5923>, rotation=<rot 75.9352 -0.0001 175.5622>}]
 clavicle_r [{shapeType=Capsule, length=7.7636, radius=11.1427, center=<-9.9642 0.7741 1.5923>, rotation=<rot 75.9351 -0.0001 -4.4378>}]
 dyn_holster [{
   shapeType=Capsule
   length=11.75
   radius=10.25
   shapeName=dyn_holster_capsule
   center=<13.5189 0 -0.6535>
   rotation=<rot -80 -179.9999 -179.9999>
  }]
 dyn_jacket_rear_04_l [{
   shapeType=Capsule
   length=17
   radius=16.25
   shapeName=dyn_jacket_rear_04_l_capsule
   center=<-1.902 5.2737 -0.5626>
   rotation=<rot -90 179.8945 -179.8943>
  }]
 dyn_jacket_rear_04_r [{
   shapeType=Capsule
   length=18.75
   radius=16.75
   shapeName=dyn_jacket_rear_04_r_capsule
   center=<-2.0871 5.128 0.9149>
   rotation=<rot -90 -90 90.0003>
  }]
 dyn_pouch_r [{shapeType=Capsule, length=6.25, radius=8.75, shapeName=dyn_pouch_r_capsule, center=<1.8938 0 0>, rotation=<rot 90 179.9736 179.9737>}]
 dyn_rope_03 [{
   shapeType=Capsule
   length=15.25
   radius=11.25
   shapeName=dyn_rope_03_capsule
   center=<-2.1642 -5.8194 7.2227>
   rotation=<rot 90 179.9996 179.9996>
  }]
 foot_l []
 foot_r []
 head [{shapeType=Sphere, radius=17.3125, center=<11.028 -0.174 0>}]
 lowerarm_l []
 lowerarm_r []
 neck_02 [{shapeType=Capsule, length=13.4375, radius=7.3175, center=<-0.2642 -0.3408 0>, rotation=<rot 90 -171.1193 -175.3978>}]
 pelvis [{shapeType=Capsule, length=7.6562, radius=19.2888, center=<-6.7574 -0.3877 0>}]
 spine_01 [{shapeType=Capsule, length=10.7187, radius=12.4138, center=<-2.6215 -2.6317 0>}]
 spine_02 [
  {shapeType=Capsule, length=12.3437, radius=11.695, center=<-2.6215 -2.6433 0.0001>}
  {shapeType=Capsule, length=29.875, radius=11.1013, center=<2.5024 0.4992 3.9825>, rotation=<rot 90 0.0049 0.0049>}
  {shapeType=Capsule, length=29.875, radius=11.1013, center=<2.4704 0.4991 -3.9825>, rotation=<rot 90 -0.0396 -0.0396>}
 ]
 spine_03 [{shapeType=Capsule, length=10.1562, radius=10.7575, center=<-1.953 -6.7423 0.0001>}]
 spine_04 [{shapeType=Capsule, length=14.6562, radius=13.6325, center=<2.1467 -4.8714 0.0002>}]
 spine_05 [
  {shapeType=Capsule, length=14.3437, radius=20.695, center=<3.8321 -0.5043 0>}
  {shapeType=Capsule, length=12.7812, radius=17.0625, center=<0.7875 3.1143 0>}
 ]
 thigh_l []
 thigh_r []
 upperarm_l []
 upperarm_r []
}
constraints []
