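# Physics asset for the BadassCowboy medium male player body, named for cloth and likely supplying leg collision for its cloth simulation.
# It covers only the legs, with one capsule body on each thigh, calf and foot bone.
# Constraints chain foot to calf to thigh on each side with all swing and twist motion free, and nothing links the two legs.
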PhysicsAsset /Game/Characters/Player/Male/Medium/Bodies/M_MED_BadassCowboy/Meshes/M_MED_BadassCowboy_Cloth_Physics
bodies {
 calf_l [{shapeType=Capsule, length=32.8591, radius=7.9516, center=<-20.107 -1.2538 0.0804>, rotation=<rot 0 92.386 -88.9929>}]
 calf_r [{shapeType=Capsule, length=32.3636, radius=7.8524, center=<19.477 0.9667 0.0048>, rotation=<rot 0 90.4706 -88.4774>}]
 foot_l [{shapeType=Capsule, length=13.0269, radius=8.5852, center=<-2.0659 -0.2309 0.0487>, rotation=<rot 0 12.9291 -87.6292>}]
 foot_r [{shapeType=Capsule, length=13.1312, radius=8.7382, center=<2.057 0.1834 -0.0579>, rotation=<rot 0 15.0052 -87.5417>}]
 thigh_l [{shapeType=Capsule, length=25.1868, radius=10.2129, center=<-28 -0.2131 0.3068>, rotation=<rot 90 0 0>}]
 thigh_r [{shapeType=Capsule, length=25.4469, radius=10.1765, center=<28 0.3685 -0.3333>, rotation=<rot 90 0 0>}]
}
constraints [
 {
  bone1Name=calf_r
  bone2Name=thigh_r
  swing1LimitDegrees=45
  swing1Motion=Free
  swing2LimitDegrees=45
  swing2Motion=Free
  twistLimitDegrees=45
  twistMotion=Free
 }
 {
  bone1Name=foot_r
  bone2Name=calf_r
  swing1LimitDegrees=45
  swing1Motion=Free
  swing2LimitDegrees=45
  swing2Motion=Free
  twistLimitDegrees=45
  twistMotion=Free
 }
 {
  bone1Name=calf_l
  bone2Name=thigh_l
  swing1LimitDegrees=45
  swing1Motion=Free
  swing2LimitDegrees=45
  swing2Motion=Free
  twistLimitDegrees=45
  twistMotion=Free
 }
 {
  bone1Name=foot_l
  bone2Name=calf_l
  swing1LimitDegrees=45
  swing1Motion=Free
  swing2LimitDegrees=45
  swing2Motion=Free
  twistLimitDegrees=45
  twistMotion=Free
 }
]
