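# Physics asset for the BadassCowboy male player's dynamic hair and coat collar bones, with capsules on head, neck, clavicles and hat pivot.
# Hair capsules form one- and two-link chains off the head with limited swing and tight twist; chain D is left-side only.
# Seven collar boxes hinge mainly on one swing axis from tiny _MATR anchor capsules.
# Those anchors, the body capsules and hat pivot link through angularly free constraints.
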PhysicsAsset /Game/Characters/Player/Male/Medium/Bodies/M_MED_BadassCowboy/Meshes/Parts/M_MED_BadassCowboy_dyn_hair
bodies {
 clavicle_l [{shapeType=Capsule, length=5.3716, radius=4.0219, center=<17.6524 1.8396 -0.747>, rotation=<rot 0 -5.0009 89.9999>}]
 clavicle_r [{shapeType=Capsule, length=5.3716, radius=4.0219, center=<-16.6972 -1.9232 1.8526>, rotation=<rot 0 -5.0008 89.9999>}]
 dyn_collar_bk [{shapeType=Box, extentX=7.5, extentY=12.5, extentZ=3.5, center=<3.4413 0 -0.3899>, rotation=<rot -5 0 0>}]
 dyn_collar_bk_MATR [{shapeType=Capsule, length=0.505, radius=0.505, rotation=<rot 0 -90 0>}]
 dyn_collar_l_01 [{shapeType=Box, extentX=7, extentY=9.5, extentZ=3, center=<3.8801 0.0289 0.0778>, rotation=<rot -0.0189 -5.4354 -4.9811>}]
 dyn_collar_l_01_MATR [{shapeType=Capsule, length=0.505, radius=0.505, rotation=<rot 0 -90 0>}]
 dyn_collar_l_02 [{shapeType=Box, extentX=6.5, extentY=10, extentZ=2.5, center=<4.2156 -0.0295 -0.3395>, rotation=<rot 5.0188 0.4369 5.0567>}]
 dyn_collar_l_02_MATR [{shapeType=Capsule, length=0.505, radius=0.505, rotation=<rot 0 -90 0>}]
 dyn_collar_l_03 [{shapeType=Box, extentX=7, extentY=8.5, extentZ=3, center=<4.3852 1.6711 0.1817>, rotation=<rot 4.0941 10.1048 2.8727>}]
 dyn_collar_l_03_MATR [{shapeType=Capsule, length=0.505, radius=0.505, rotation=<rot 0 -90 0>}]
 dyn_collar_r_01 [{shapeType=Box, extentX=7, extentY=9.5, extentZ=3, center=<3.8739 0.0363 0.0002>, rotation=<rot 4.9621 5.8708 4.9998>}]
 dyn_collar_r_01_MATR [{shapeType=Capsule, length=0.505, radius=0.505, rotation=<rot 0 -90 0>}]
 dyn_collar_r_02 [{shapeType=Box, extentX=6.5, extentY=10, extentZ=2.5, center=<4.1132 -0.0312 -0.2012>, rotation=<rot 5.0188 -0.4369 -0.0569>}]
 dyn_collar_r_02_MATR [{shapeType=Capsule, length=0.505, radius=0.505, rotation=<rot 0 -90 0>}]
 dyn_collar_r_03 [{shapeType=Box, extentX=7, extentY=6.5, extentZ=3, center=<3.7842 -1.564 0.1381>, rotation=<rot 4.094 -10.1043 -2.8728>}]
 dyn_collar_r_03_MATR [{shapeType=Capsule, length=0.505, radius=0.505, rotation=<rot 0 -90 0>}]
 dyn_hair_A_l [{shapeType=Capsule, length=2, radius=1, center=<1 0 0>, rotation=<rot 90 0 0>}]
 dyn_hair_A_r [{shapeType=Capsule, length=2, radius=1, center=<1 0 0>, rotation=<rot 90 0 0>}]
 dyn_hair_B_01_l [{shapeType=Capsule, length=2, radius=1.5, center=<2 0 0>, rotation=<rot 90 0 0>}]
 dyn_hair_B_01_r [{shapeType=Capsule, length=2.25, radius=1.25, center=<2 0 0>, rotation=<rot 90 0 0>}]
 dyn_hair_B_02_l [{shapeType=Capsule, length=2, radius=1.5, center=<2 0 0>, rotation=<rot 90 0 0>}]
 dyn_hair_B_02_r [{shapeType=Capsule, length=2.25, radius=1.25, center=<1.9569 0 -0.2446>, rotation=<rot 79.9999 0 0>}]
 dyn_hair_C_01_l [{shapeType=Capsule, length=2, radius=1.5, center=<2 0 0>, rotation=<rot 90 0 0>}]
 dyn_hair_C_01_r [{shapeType=Capsule, length=2, radius=1.5, center=<2 0 0>, rotation=<rot 90 0 0>}]
 dyn_hair_C_02_l [{shapeType=Capsule, length=2, radius=1.5, center=<2 0 0>, rotation=<rot 90 0 0>}]
 dyn_hair_C_02_r [{shapeType=Capsule, length=2, radius=1.5, center=<2 0 0>, rotation=<rot 90 0 0>}]
 dyn_hair_D_01_l [{shapeType=Capsule, length=2.25, radius=1.25, center=<2 0 0>, rotation=<rot 90 0 0>}]
 dyn_hair_D_02_l [{shapeType=Capsule, length=2.25, radius=1.25, center=<1.9872 -0.3725 0.0001>, rotation=<rot 90 -6.5366 -1.5367>}]
 dyn_hat_pivot [{shapeType=Capsule, length=1.0936, radius=5.937, center=<-0.3181 3.564 -5.2813>, rotation=<rot 2.6228 -89.9999 0.0001>}]
 head [
  {shapeType=Capsule, length=4.75, radius=8, center=<5.7408 -0.0797 0>, rotation=<rot 90 0 0>}
  {shapeType=Capsule, length=6.75, radius=6, center=<3.5326 6.7325 -0.0001>, rotation=<rot 90 0 0>}
 ]
 neck_01 [{shapeType=Capsule, length=0.5, radius=4.75, rotation=<rot 90 0 0>}]
 neck_02 [{shapeType=Capsule, length=2.25, radius=5.25, rotation=<rot 90 0 0>}]
 spine_05 [{shapeType=Capsule, length=0.5, radius=0.5}]
}
constraints [
 {
  bone1Name=dyn_collar_bk_MATR
  bone2Name=spine_05
  swing1LimitDegrees=45
  swing1Motion=Free
  swing2LimitDegrees=45
  swing2Motion=Free
  twistLimitDegrees=45
  twistMotion=Free
 }
 {bone1Name=dyn_collar_bk, bone2Name=dyn_collar_bk_MATR, swing1LimitDegrees=5, swing2LimitDegrees=75, twistLimitDegrees=2}
 {
  bone1Name=dyn_collar_r_01_MATR
  bone2Name=spine_05
  swing1LimitDegrees=45
  swing1Motion=Free
  swing2LimitDegrees=45
  swing2Motion=Free
  twistLimitDegrees=45
  twistMotion=Free
 }
 {bone1Name=dyn_collar_r_01, bone2Name=dyn_collar_r_01_MATR, swing1LimitDegrees=5, swing2LimitDegrees=75, twistLimitDegrees=2}
 {
  bone1Name=dyn_collar_r_02_MATR
  bone2Name=spine_05
  swing1LimitDegrees=45
  swing1Motion=Free
  swing2LimitDegrees=45
  swing2Motion=Free
  twistLimitDegrees=45
  twistMotion=Free
 }
 {bone1Name=dyn_collar_r_02, bone2Name=dyn_collar_r_02_MATR, swing1LimitDegrees=5, swing2LimitDegrees=45, twistLimitDegrees=2}
 {
  bone1Name=dyn_collar_r_03_MATR
  bone2Name=spine_05
  swing1LimitDegrees=45
  swing1Motion=Free
  swing2LimitDegrees=45
  swing2Motion=Free
  twistLimitDegrees=45
  twistMotion=Free
 }
 {bone1Name=dyn_collar_r_03, bone2Name=dyn_collar_r_03_MATR, swing1LimitDegrees=5, swing2LimitDegrees=45, twistLimitDegrees=2}
 {
  bone1Name=dyn_collar_l_01_MATR
  bone2Name=spine_05
  swing1LimitDegrees=45
  swing1Motion=Free
  swing2LimitDegrees=45
  swing2Motion=Free
  twistLimitDegrees=45
  twistMotion=Free
 }
 {bone1Name=dyn_collar_l_01, bone2Name=dyn_collar_l_01_MATR, swing1LimitDegrees=5, swing2LimitDegrees=75, twistLimitDegrees=2}
 {
  bone1Name=dyn_collar_l_02_MATR
  bone2Name=spine_05
  swing1LimitDegrees=45
  swing1Motion=Free
  swing2LimitDegrees=45
  swing2Motion=Free
  twistLimitDegrees=45
  twistMotion=Free
 }
 {bone1Name=dyn_collar_l_02, bone2Name=dyn_collar_l_02_MATR, swing1LimitDegrees=5, swing2LimitDegrees=45, twistLimitDegrees=2}
 {
  bone1Name=dyn_collar_l_03_MATR
  bone2Name=spine_05
  swing1LimitDegrees=45
  swing1Motion=Free
  swing2LimitDegrees=45
  swing2Motion=Free
  twistLimitDegrees=45
  twistMotion=Free
 }
 {bone1Name=dyn_collar_l_03, bone2Name=dyn_collar_l_03_MATR, swing1LimitDegrees=5, swing2LimitDegrees=45, twistLimitDegrees=2}
 {
  bone1Name=neck_01
  bone2Name=spine_05
  swing1LimitDegrees=45
  swing1Motion=Free
  swing2LimitDegrees=45
  swing2Motion=Free
  twistLimitDegrees=45
  twistMotion=Free
 }
 {
  bone1Name=neck_02
  bone2Name=neck_01
  swing1LimitDegrees=45
  swing1Motion=Free
  swing2LimitDegrees=45
  swing2Motion=Free
  twistLimitDegrees=45
  twistMotion=Free
 }
 {
  bone1Name=head
  bone2Name=neck_02
  swing1LimitDegrees=45
  swing1Motion=Free
  swing2LimitDegrees=45
  swing2Motion=Free
  twistLimitDegrees=45
  twistMotion=Free
 }
 {
  bone1Name=clavicle_l
  bone2Name=spine_05
  swing1LimitDegrees=45
  swing1Motion=Free
  swing2LimitDegrees=45
  swing2Motion=Free
  twistLimitDegrees=45
  twistMotion=Free
 }
 {
  bone1Name=clavicle_r
  bone2Name=spine_05
  swing1LimitDegrees=45
  swing1Motion=Free
  swing2LimitDegrees=45
  swing2Motion=Free
  twistLimitDegrees=45
  twistMotion=Free
 }
 {bone1Name=dyn_hair_A_r, bone2Name=head, swing1LimitDegrees=45, swing2LimitDegrees=45, twistLimitDegrees=5}
 {bone1Name=dyn_hair_B_01_r, bone2Name=head, swing1LimitDegrees=45, swing2LimitDegrees=45, twistLimitDegrees=5}
 {bone1Name=dyn_hair_B_02_r, bone2Name=dyn_hair_B_01_r, swing1LimitDegrees=45, swing2LimitDegrees=45, twistLimitDegrees=5}
 {bone1Name=dyn_hair_C_01_r, bone2Name=head, swing1LimitDegrees=45, swing2LimitDegrees=45, twistLimitDegrees=5}
 {bone1Name=dyn_hair_C_02_r, bone2Name=dyn_hair_C_01_r, swing1LimitDegrees=45, swing2LimitDegrees=45, twistLimitDegrees=5}
 {bone1Name=dyn_hair_A_l, bone2Name=head, swing1LimitDegrees=45, swing2LimitDegrees=45, twistLimitDegrees=5}
 {bone1Name=dyn_hair_B_01_l, bone2Name=head, swing1LimitDegrees=45, swing2LimitDegrees=45, twistLimitDegrees=5}
 {bone1Name=dyn_hair_B_02_l, bone2Name=dyn_hair_B_01_l, swing1LimitDegrees=45, swing2LimitDegrees=45, twistLimitDegrees=5}
 {bone1Name=dyn_hair_C_01_l, bone2Name=head, swing1LimitDegrees=45, swing2LimitDegrees=45, twistLimitDegrees=5}
 {bone1Name=dyn_hair_C_02_l, bone2Name=dyn_hair_C_01_l, swing1LimitDegrees=45, swing2LimitDegrees=45, twistLimitDegrees=5}
 {bone1Name=dyn_hair_D_01_l, bone2Name=head, swing1LimitDegrees=45, swing2LimitDegrees=45, twistLimitDegrees=5}
 {bone1Name=dyn_hair_D_02_l, bone2Name=dyn_hair_D_01_l, swing1LimitDegrees=45, swing2LimitDegrees=45, twistLimitDegrees=5}
 {
  bone1Name=dyn_hat_pivot
  bone2Name=head
  swing1LimitDegrees=45
  swing1Motion=Free
  swing2LimitDegrees=45
  swing2Motion=Free
  twistLimitDegrees=45
  twistMotion=Free
 }
]
